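# Niagara transition effect for the FluteLamp back bling: a short burst of diamond sprites, toon smoke, a toon shape and flying "toppings" that scales with User.ElimCount.
# Arrays indexed mostly by ElimCount-1 set counts, offsets, heights and speeds per tier, with the jittering toon shape reserved for the top tier.
# Diamonds, smoke and shape sample a socket reader.
# Toppings mix spinning sprites and meshes under gravity, tinted by time of day.
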NiagaraSystem /BRCosmetics/Accessories/FORT_Backpacks/Backpack_M_MED_FluteLamp/FX/NS_Backpack_FluteLamp_Transition
dependencies {
 usedDataInterfaces [
  &NiagaraDataInterfaceArrayInt32
  &NiagaraDataInterfaceArrayFloat
  &NiagaraDataInterfaceArrayFloat3
  &NiagaraDataInterfaceVector2DCurve
  &NiagaraDataInterfaceCurve
  &NiagaraDataInterfaceSocketReader
 ]
 usedDynamicInputs [&MakeCustomVectorFromBool, &SetBoolByIntComparison]
 usedModules [
  &self:SystemSpawnScript.SystemScriptSource.SystemScriptGraph.NiagaraNodeAssignment_1.SetVariables_#1
  &SystemState
 ]
}
emitters {
 Diamond_Sprites {
  emitterUpdateScript [
   &EmitterState()
   &SpawnBurst_Instantaneous(
    Spawn Count=&SelectIntFromArray(
     Direct Array Index=&Add_Integer(A=$User.ElimCount, B=-1)
     Int Selection Array=DataInterface() {propertyValues {IntData [6,6,10,12]}}
    )
    Use Spawn Probability=false
    UseLoopCountLimit=false
   )
  ]
  particleSpawnScript [
   &InitializeParticle(
    Color=<rgba 0.651 0.1734 0.0203 30>
    Lifetime Max=&SelectFloatFromArray(Direct Array Index=$User.ElimCount, Float Selection Array=DataInterface() {propertyValues {FloatData [0,0.6,0.8,0.9,1.2]}})
    Lifetime Min=0.5
    Lifetime Mode=::NewEnumerator1
    Sprite Size Mode=::NewEnumerator1
    Uniform Sprite Size Max=6
    Uniform Sprite Size Min=3
    UsePositionOffset=false
    Write Material Random=true
   ) {module=InitializeParticle001}
   &SocketLocation(
    SamplePosOffset=true
    Sampled Position Offset=&SelectVectorFromArray(
     Direct Array Index=&Add_Integer(A=$User.ElimCount, B=-1)
     Vector Selection Array=DataInterface() {propertyValues {FloatData [<-3 0 0>, <-8 0 0>, <-13 0 0>, <-18 0 0>]}}
    )
    Socket Reader=$System.Socket Reader
    Socket Sampling Mode=::NewEnumerator3
    VelocitySpeedLimit=false
   )
   &ShapeLocation(
    Cylinder Distribution=::NewEnumerator2
    Cylinder Height=&SelectFloatFromArray(
     Direct Array Index=&Add_Integer(A=$User.ElimCount, B=-1)
     Float Selection Array=DataInterface() {propertyValues {FloatData [6,10,18,20]}}
    )
    Cylinder Radius=&RandomRangeFloat(Minimum=10, Maximum=12.5)
    MaskSpawnGroup=false
    Shape Primitive=::NewEnumerator1
    U Position=&Multiply_Float(
     A=&ReturnNormalizedExecIndex()
     B=&RandomRangeFloat(Minimum=0.9, Maximum=1.1)
    )
    UseOwnerScale=true
    V Position=&RandomRangeFloat(Minimum=0, Maximum=1)
   )
   &CameraOffset(Camera Offset Amount=5)
   &AddVelocity(
    Velocity=<0 0 1>
    Velocity Speed Scale=&Multiply_Float(
     A=&RandomRangeFloat(Minimum=20, Maximum=60)
     B=&SelectFloatFromArray(
      Direct Array Index=&Add_Integer(A=$User.ElimCount, B=-1)
      Float Selection Array=DataInterface() {propertyValues {FloatData [0.6,0.7,0.9,1.5]}}
     )
    )
   )
  ]
  particleUpdateScript [
   &ParticleState()
   &Drag(Drag=3, Use Linear Drag=true, Use Rotational Drag=true)
   &SolveForcesAndVelocity(Clamp Velocity=false, Limit Acceleration=false)
   &ScaleSpriteSize(Scale Factor=&Vector2DFromCurve(
     Vector2Curve=DataInterface() {propertyValues {
       XCurve {
        defaultValue=0
        keys [
         {arriveTangent=3.418, interpMode=RCIM_Cubic, leaveTangent=3.418, tangentMode=RCTM_User, time=0, value=0.6}
         {interpMode=RCIM_Cubic, tangentMode=RCTM_User, time=1, value=0}
        ]
       }
       YCurve {
        defaultValue=0
        keys [
         {arriveTangent=3.418, interpMode=RCIM_Cubic, leaveTangent=3.418, tangentMode=RCTM_User, time=0, value=0.6}
         {interpMode=RCIM_Cubic, tangentMode=RCTM_User, time=1, value=0}
        ]
       }
      }}
     ScaleCurve=&Vector2DFromFloat(Value=1)
    ))
   &ScaleColor(
    Scale Alpha=&FloatFromCurve(
     FloatCurve=DataInterface() {propertyValues {Curve {keys [
         {arriveTangent=2.1312, interpMode=RCIM_Cubic, leaveTangent=2.1312, tangentMode=RCTM_User, time=0, value=0.6}
         {interpMode=RCIM_Cubic, tangentMode=RCTM_User, time=0.5, value=1}
         {arriveTangent=-2.8077, tangentMode=RCTM_User, time=1, value=0}
        ]}}}
     Scale Curve=0.2
    )
    Scale RGB=&VectorFromFloat(Value=&FloatFromCurve(FloatCurve=DataInterface() {propertyValues {Curve {keys [{interpMode=RCIM_Cubic, tangentMode=RCTM_User, time=0, value=1}]}}}))
    ScaleA=true
    ScaleRGB=true
   )
   &FN_Sprite_Attributes(
    Color=<rgba 1.5 0.7632 0.075 10>
    Create Color Attribute=true
    Create SpriteSize Attribute=true
    SpriteSize=&Multiply_Vector2D_ByFloat(Float=0.65, Vector2D=$Particles.SpriteSize)
    Write to Module Namespace=true
   )
  ]
  rendererClasses [&NiagaraSpriteRendererProperties]
  renderers [{rendererClass=&NiagaraSpriteRendererProperties}]
 }
 Toon_Shapes {
  emitterUpdateScript [
   &EmitterState()
   &SpawnBurst_Instantaneous(
    Spawn Count=&SelectIntFromArray(
     Direct Array Index=&Add_Integer(A=$User.ElimCount, B=-1)
     Int Selection Array=DataInterface() {propertyValues {IntData [0,0,0,1]}}
    )
    Use Spawn Probability=false
    UseLoopCountLimit=false
   )
  ]
  particleSpawnScript [
   &InitializeParticle(
    Color=<rgba 1 0.4334 0 30>
    Lifetime=1.5
    Position Mode=::NewEnumerator0
    Sprite Rotation Mode=::NewEnumerator0
    Sprite Size Mode=::NewEnumerator3
    Write Material Random=true
   )
   &CameraOffset(Camera Offset Amount=10)
   /SetVariables(Particles.PositionForRotation:NiagaraPosition=<0 0 0>) {isSetParameters=true}
   &AddVelocity(Velocity=<-1 0 1.5>)
  ]
  particleUpdateScript [
   &ParticleState()
   &ScaleSpriteSize(
    Scale Sprite Size Mode=::NewEnumerator0
    Uniform Scale Factor=&FloatFromCurve(FloatCurve=DataInterface() {propertyValues {Curve {
        defaultValue=0.5
        keys [
         {arriveTangent=10.9569, interpMode=RCIM_Cubic, leaveTangent=10.9569, tangentMode=RCTM_User, time=0, value=0}
         {arriveTangent=-0.1259, interpMode=RCIM_Cubic, leaveTangent=-0.1259, tangentMode=RCTM_User, time=0.2, value=1}
         {interpMode=RCIM_Cubic, tangentMode=RCTM_User, time=0.8, value=1}
         {arriveTangent=-13.4303, interpMode=RCIM_Cubic, leaveTangent=-13.4303, tangentMode=RCTM_User, time=1, value=0}
        ]
       }}})
   )
   &ScaleColor(ScaleA=true, ScaleRGB=true)
   &DynamicMaterialParameters(
    Index 0 Param 1=1
    Index 0 Param 2=1.5
    Index 0 Param 3=&RandomRangeFloat(Minimum=0, Maximum=1)
    Module.Index0Float0Write=true
    Module.Index0Float1Write=true
    Module.Index0Float2Write=true
    Module.Index0Float3Write=true
   )
   &InterpolateOverTime(
    Rate Of Change:NiagaraFloat=&FN_FrontEndFloatModifier(Gameplay=20, Front End=5)
    Target Vector Value:Vector3f=&TransformPosition(Position=/Niagara/DynamicInputs/Helpers/SimulationPosition(), Source Space=::Simulation, Destination Space=::World)
    Variable Type=::NewEnumerator2
   )
   &SocketLocation(
    SamplePosOffset=true
    Sampled Position Offset=<6 8 0>
    Socket Reader=$System.Socket Reader
    Socket Sampling Mode=::NewEnumerator3
    VelocitySpeedLimit=false
   )
   &OffsetPosition(Jitter Amount=0.5, Jitter Delay=0.1, Jitter Mode=1, Offset Mode=1)
   &OffsetPosition(
    Jitter Amount=0.2
    Jitter Delay=0.1
    Jitter Mode=1
    Offset Mode=1
    Position Output=1
    Position to Offset:NiagaraPosition=$Particles.PositionForRotation
   ) {module=OffsetPosition001}
   /SetVariables(
    Particles.SpriteRotation:NiagaraFloat=&Add_Float(
     A=$Particles.Initial.SpriteRotation
     B=&MakeFloatFromVector(Vector=Linked() {linkedVariable {name=Particles.OffsetPosition001.Position, type=NiagaraPosition}}, Channel=::NewEnumerator0)
    )
    Particles.Velocity=&Add_Vector(A=$Particles.Initial.Velocity, B=Linked() {linkedVariable {name=Particles.OffsetPosition001.Position, type=NiagaraPosition}})
   ) {isSetParameters=true}
   &Drag(Drag=3, Use Linear Drag=true, Use Rotational Drag=true)
   &SolveForcesAndVelocity(Clamp Velocity=false, Limit Acceleration=false)
  ]
  rendererClasses [&NiagaraSpriteRendererProperties]
  renderers [{rendererClass=&NiagaraSpriteRendererProperties}]
 }
 Toon_Smoke {
  emitterUpdateScript [
   &EmitterState()
   &SpawnBurst_Instantaneous(Use Spawn Probability=false, UseLoopCountLimit=false)
  ]
  particleSpawnScript [
   /SetVariables(Particles.Random=&RandomRangeFloat(Minimum=0, Maximum=1)) {isSetParameters=true}
   &PartitionParticles(Particles Per Partition=1, Partition Mode:ENiagara_PartitionMode_v2=::NewEnumerator1)
   &InitializeParticle(
    Lifetime=&RandomRangeFloat(Minimum=0.6, Maximum=0.8)
    Sprite Size Mode=::NewEnumerator1
    Sprite UV Mode=::NewEnumerator1
    Uniform Sprite Size Max=60
    Uniform Sprite Size Min=50
    UsePositionOffset=false
    Write Material Random=true
   ) {module=InitializeParticle001}
   &SocketLocation(
    SamplePosOffset=true
    Sampled Position Offset=&SelectVectorFromArray(
     Direct Array Index=&Add_Integer(A=$User.ElimCount, B=-1)
     Vector Selection Array=DataInterface() {propertyValues {FloatData [<0 0 0>, <1.5 0 0>, <5 0 0>, <11 0 0>]}}
    )
    Socket Reader=$System.Socket Reader
    Socket Sampling Mode=::NewEnumerator3
    VelocitySpeedLimit=false
   )
   &CameraOffset(Camera Offset Amount=30)
   &AddVelocity(
    Cone Angle=10
    Cone Axis=<0 0 1>
    Module.UseConeFalloff=false
    Module.UseVelDistribution=false
    Override Seed=false
    Velocity Mode=::NewEnumerator2
    Velocity Speed=&RandomRangeFloat(Minimum=35, Maximum=50)
   )
  ]
  particleUpdateScript [
   &ParticleState()
   &ScaleSpriteSize(Scale Sprite Size Mode=::NewEnumerator0, Uniform Scale Factor=0.8)
   &Drag(Drag=5, Use Linear Drag=true, Use Rotational Drag=true)
   &SolveForcesAndVelocity(Clamp Velocity=false, Limit Acceleration=false)
   &DynamicMaterialParameters(
    Index 0 Param 1=&FloatFromCurve(FloatCurve=DataInterface() {propertyValues {Curve {
        defaultValue=0
        keys [
         {arriveTangent=1.5524, interpMode=RCIM_Cubic, leaveTangent=1.5524, tangentMode=RCTM_User, time=0, value=0}
         {interpMode=RCIM_Cubic, tangentMode=RCTM_User, time=1, value=1}
        ]
       }}})
    Index 0 Param 2=1
    Module.Index0Float0Write=true
    Module.Index0Float1Write=true
    Module.Index0Float2Write=true
    Module.Index0Float3Write=true
   )
   &ScaleColor(
    Scale Alpha=&FloatFromCurve(
     FloatCurve=DataInterface() {propertyValues {Curve {keys [{interpMode=RCIM_Cubic, tangentMode=RCTM_User, time=0, value=1}, {interpMode=RCIM_Cubic, tangentMode=RCTM_User, time=1, value=0}]}}}
     Scale Curve=0.5
    )
    Scale RGB=&VectorFromFloat(Value=&FNTimeOfDayScalar(Night Scale=0.6, Morning Scale=0.75, Evening Scale=0.75))
    ScaleA=true
    ScaleRGB=true
   )
  ]
  rendererClasses [&NiagaraSpriteRendererProperties]
  renderers [{rendererClass=&NiagaraSpriteRendererProperties}]
 }
 Toppings {
  emitterUpdateScript [
   &EmitterState()
   &SpawnRate(
    SpawnRate=&FloatFromCurve(
     FloatCurve=DataInterface() {propertyValues {Curve {
        defaultValue=1
        keys [
         {arriveTangent=-0.031, interpMode=RCIM_Cubic, leaveTangent=-0.031, tangentMode=RCTM_User, time=0, value=1}
         {arriveTangent=-3.7858, interpMode=RCIM_Cubic, leaveTangent=-3.7858, tangentMode=RCTM_User, time=0.5, value=0}
        ]
       }}}
     CurveIndex=$Emitter.NormalizedLoopAge
     Scale Curve=&SelectFloatFromArray(
      Direct Array Index=&Add_Integer(A=$User.ElimCount, B=-1)
      Float Selection Array=DataInterface() {propertyValues {FloatData [10,20,30,30]}}
     )
    )
    Use Spawn Probability=false
   )
   &SpawnBurst_Instantaneous(
    Spawn Count=&SelectIntFromArray(
     Direct Array Index=&Add_Integer(A=$User.ElimCount, B=-1)
     Int Selection Array=DataInterface() {propertyValues {IntData [5,7,9,9]}}
    )
    Use Spawn Probability=false
    UseLoopCountLimit=false
   )
  ]
  particleSpawnScript [
   /SetVariables(Particles.SubImageIndex=&RandomRangeFloat(Minimum=0, Maximum=8)) {isSetParameters=true}
   /SetVariables(Particles.MeshIndex=&RandomRangeInt(Minimum=0, Maximum=3)) {isSetParameters=true}
   &InitializeParticle(
    Color Mode=::NewEnumerator0
    Lifetime Max=0.6
    Lifetime Min=0.4
    Lifetime Mode=::NewEnumerator1
    Mesh Scale Mode=::NewEnumerator3
    Mesh Uniform Scale=&SelectFloatFromArray(Direct Array Index=$Particles.MeshIndex, Float Selection Array=DataInterface() {propertyValues {FloatData [1,0.3,0.5,0.5]}})
    Position Offset=&RandomRangeVector(Minimum=<0 -10 0>, Maximum=<0 -10 -5>)
    Position Offset Coordinate Space=::Local
    Sprite Rotation Mode=::NewEnumerator2
    Sprite Size=&MakeCustomVector2DFromBool(
     Bool=&SetBoolByFloatComparison(A=$Particles.SubImageIndex, B=2, Comparison Type=::NewEnumerator7)
     True Vector2D=<15 15>
     False Vector2D=&MakeCustomVector2DFromBool(
      Bool=&SetBoolByFloatComparison(A=$Particles.SubImageIndex, B=5, Comparison Type=::NewEnumerator7)
      True Vector2D=<12 12>
      False Vector2D=<13 13>
     )
    )
    Sprite Size Mode=::NewEnumerator4
    UsePositionOffset=true
    Write Material Random=true
   )
   &ShapeLocation(
    Cylinder Height=&SelectFloatFromArray(
     Direct Array Index=&Add_Integer(A=$User.ElimCount, B=-1)
     Float Selection Array=DataInterface() {propertyValues {FloatData [6,10,15,22]}}
    )
    Cylinder Radius=10
    MaskSpawnGroup=false
    Override Seed=false
    Shape Primitive=::NewEnumerator1
    Surface Only=true
    Use Endcaps In Surface Only Mode=false
    UseOwnerScale=true
   )
   &AddVelocity(
    Module.Use Falloff Distance=false
    Origin Offset=<0 0 -20>
    Override Seed=false
    Velocity Mode=::NewEnumerator1
    Velocity Speed=&Multiply_Float(
     A=&RandomRangeFloat(Minimum=75, Maximum=100)
     B=&SelectFloatFromArray(
      Direct Array Index=&Add_Integer(A=$User.ElimCount, B=-1)
      Float Selection Array=DataInterface() {propertyValues {FloatData [0.6,0.6,0.9,1]}}
     )
    )
   )
   &InheritVelocity(
    Inherited Velocity Speed Limit:NiagaraFloat=&RandomRangeFloat(Minimum=50, Maximum=100)
    Speed Limit=true
    UseSpeedThreshold=false
   )
   /SetVariables(Particles.VisibilityTag=&RandomRangeInt(Minimum=0, Maximum=1)) {isSetParameters=true}
  ]
  particleUpdateScript [
   &ParticleState()
   &GravityForce(Gravity=&Add_Vector(
     A=&TransformVector(Vector=<0 -30 0>, Destination Space=::Simulation)
     B=&RandomRangeVector(Minimum=<0 0 -200>, Maximum=<0 0 -300>)
    ))
   &Drag(Drag=3, Use Linear Drag=true, Use Rotational Drag=false)
   &SolveForcesAndVelocity(Clamp Velocity=false, Limit Acceleration=false)
   &ScaleSpriteSize(
    Uniform Curve Scale=0.8
    Uniform Curve Sprite Scale=DataInterface() {propertyValues {Curve {keys [
        {interpMode=RCIM_Cubic, tangentMode=RCTM_User, time=0, value=1}
        {arriveTangent=-1.8584, interpMode=RCIM_Cubic, leaveTangent=-1.8584, tangentMode=RCTM_User, time=1, value=0}
       ]}}}
   )
   &ScaleMeshSize(Scale Factor=&VectorFromFloat(Value=&FloatFromCurve(
      FloatCurve=DataInterface() {propertyValues {Curve {keys [
          {interpMode=RCIM_Cubic, tangentMode=RCTM_User, time=0, value=1}
          {arriveTangent=-1.8584, interpMode=RCIM_Cubic, leaveTangent=-1.8584, tangentMode=RCTM_User, time=1, value=0}
         ]}}}
      Scale Curve=0.8
     )))
   &SpriteRotationRate(Rotation Rate=&Multiply_FloatByInt(
     Float=&RandomRangeFloat(Minimum=300, Maximum=500)
     Integer=&SelectIntFromArray(
      Array Sampling Mode=::NewEnumerator1
      Direct Array Index=?
      Int Selection Array=DataInterface() {propertyValues {IntData [-1,1]}}
      Random Seed=0
     )
    ))
   &UpdateMeshOrientation(
    Rotation Rate=&FloatFromCurve(
     FloatCurve=DataInterface() {propertyValues {Curve {
        defaultValue=1
        keys [{interpMode=RCIM_Cubic, tangentMode=RCTM_User, time=0, value=1}, {interpMode=RCIM_Cubic, tangentMode=RCTM_User, time=1, value=0}]
       }}}
     Scale Curve=3
    )
    Rotation Vector=&RandomVector(Vector Scale=1)
   )
   &ScaleColor(
    Scale RGB=&MakeVectorFromLinearColorRGB(Color=&FNTimeOfDayColor(
      Night Color=<rgba 0.4516 0.599 0.85 1>
      Morning Color=<rgba 0.675 0.7045 0.9 1>
      Day Color=<rgba 1 1 1 1>
      Evening Color=<rgba 0.85 0.6818 0.6818 1>
     ))
    ScaleA=true
    ScaleRGB=true
   )
  ]
  rendererClasses [&NiagaraSpriteRendererProperties, &NiagaraMeshRendererProperties]
  renderers [
   {rendererClass=&NiagaraSpriteRendererProperties}
   {rendererClass=&NiagaraMeshRendererProperties, rendererIndex=1}
  ]
 }
}
summary {
 systemName=NS_Backpack_FluteLamp_Transition
 emitters [
  {emitterName=Diamond_Sprites, rendererClasses [&NiagaraSpriteRendererProperties]}
  {emitterName=Toon_Shapes, rendererClasses [&NiagaraSpriteRendererProperties]}
  {emitterName=Toppings, rendererClasses [&NiagaraSpriteRendererProperties, &NiagaraMeshRendererProperties]}
  {emitterName=Toon_Smoke, rendererClasses [&NiagaraSpriteRendererProperties]}
 ]
 userVariables [
  User.bIsFrontEnd {type=NiagaraBool}
  User.bIsFrontEndPreview {type=NiagaraBool}
  User.ElimCount {type=NiagaraInt32, defaultValue=5}
  User.SwingSectionIndex {type=NiagaraInt32}
 ]
}
