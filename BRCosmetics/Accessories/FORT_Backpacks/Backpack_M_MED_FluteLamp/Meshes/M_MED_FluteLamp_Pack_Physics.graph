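# Physics asset for the M_MED_FluteLamp backpack cosmetic, defining a jointed chain of capsule bodies on bones named as burger layers.
# A small base capsule anchors nine identical, larger dynamic-layer capsules, each constrained to the one before it.
# Swing1 limits widen gradually up the chain before levelling off, while swing2 and twist limits are zero, so the stack bends on one axis and more loosely toward the tip.
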PhysicsAsset /BRCosmetics/Accessories/FORT_Backpacks/Backpack_M_MED_FluteLamp/Meshes/M_MED_FluteLamp_Pack_Physics
bodies {
 base_burger_layer_01 [{shapeType=Capsule, length=0.505, radius=0.505, shapeName=base_burger_layer_01_capsule, rotation=<rot 0 -90 0>}]
 dyn_burger_layer_01 [{shapeType=Capsule, length=4.7021, radius=1.9979, shapeName=dyn_burger_layer_01_capsule, center=<1.9472 0 0>, rotation=<rot 90 0 0>}]
 dyn_burger_layer_02 [{shapeType=Capsule, length=4.7021, radius=1.9979, shapeName=dyn_burger_layer_02_capsule, center=<1.9472 0 0>, rotation=<rot 90 0 0>}]
 dyn_burger_layer_03 [{shapeType=Capsule, length=4.7021, radius=1.9979, shapeName=dyn_burger_layer_03_capsule, center=<1.9472 0 0>, rotation=<rot 90 0 0>}]
 dyn_burger_layer_04 [{shapeType=Capsule, length=4.7021, radius=1.9979, shapeName=dyn_burger_layer_04_capsule, center=<1.9472 0 0>, rotation=<rot 90 0 0>}]
 dyn_burger_layer_05 [{shapeType=Capsule, length=4.7021, radius=1.9979, shapeName=dyn_burger_layer_05_capsule, center=<1.9472 0 0>, rotation=<rot 90 0 0>}]
 dyn_burger_layer_06 [{shapeType=Capsule, length=4.7021, radius=1.9979, shapeName=dyn_burger_layer_06_capsule, center=<1.9472 0 0>, rotation=<rot 90 0 0>}]
 dyn_burger_layer_07 [{shapeType=Capsule, length=4.7021, radius=1.9979, shapeName=dyn_burger_layer_07_capsule, center=<1.9472 0 0>, rotation=<rot 90 0 0>}]
 dyn_burger_layer_08 [{shapeType=Capsule, length=4.7021, radius=1.9979, shapeName=dyn_burger_layer_08_capsule, center=<1.9472 0 0>, rotation=<rot 90 0 0>}]
 dyn_burger_layer_09 [{shapeType=Capsule, length=4.7021, radius=1.9979, shapeName=dyn_burger_layer_09_capsule, center=<1.9472 0 0>, rotation=<rot 90 0 0>}]
}
constraints [
 {bone1Name=dyn_burger_layer_01, bone2Name=base_burger_layer_01, swing1LimitDegrees=2, swing2LimitDegrees=0, twistLimitDegrees=0}
 {bone1Name=dyn_burger_layer_02, bone2Name=dyn_burger_layer_01, swing1LimitDegrees=3, swing2LimitDegrees=0, twistLimitDegrees=0}
 {bone1Name=dyn_burger_layer_03, bone2Name=dyn_burger_layer_02, swing1LimitDegrees=4, swing2LimitDegrees=0, twistLimitDegrees=0}
 {bone1Name=dyn_burger_layer_04, bone2Name=dyn_burger_layer_03, swing1LimitDegrees=5, swing2LimitDegrees=0, twistLimitDegrees=0}
 {bone1Name=dyn_burger_layer_05, bone2Name=dyn_burger_layer_04, swing1LimitDegrees=6, swing2LimitDegrees=0, twistLimitDegrees=0}
 {bone1Name=dyn_burger_layer_06, bone2Name=dyn_burger_layer_05, swing1LimitDegrees=7, swing2LimitDegrees=0, twistLimitDegrees=0}
 {bone1Name=dyn_burger_layer_07, bone2Name=dyn_burger_layer_06, swing1LimitDegrees=8, swing2LimitDegrees=0, twistLimitDegrees=0}
 {bone1Name=dyn_burger_layer_08, bone2Name=dyn_burger_layer_07, swing1LimitDegrees=8, swing2LimitDegrees=0, twistLimitDegrees=0}
 {bone1Name=dyn_burger_layer_09, bone2Name=dyn_burger_layer_08, swing1LimitDegrees=8, swing2LimitDegrees=0, twistLimitDegrees=0}
]
