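# Animation Blueprint for the FluteLamp backpack, a stack of burger-layer bones with staged idle and kill-reaction animations.
# CosmeticVariantReactivity switches between an Idle/OnKill state machine and the plain stage 00 idle.
# Idle layers spring-based Anim Dynamics onto stage idles, using RigidBody for stage 04, weighted by the DynamicsAlpha curve.
# An unwired duplicate state machine picks stages by KillCounter and triggers on ActiveKill; a nine-bone dynamics chain is also unwired.
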AnimBlueprint /BRCosmetics/Accessories/FORT_Backpacks/Backpack_M_MED_FluteLamp/Meshes/M_MED_FluteLamp_Pack_AnimBP
target_skeleton=&/M_MED_FluteLamp_Pack_Skeleton
graphs [
 AnimGraph {nodes [
   n1 = Anim Dynamics - Bone: dyn_burger_layer_01!(ComponentPose, Alpha=1, GravityOverride:Vector="0,0,980", ExternalForce="0,0,0") -> (Pose) {
    angularDampingOverride=0.9999
    angularSpringConstant=7000
    bAngularSpring=true
    bOverrideAngularDamping=true
    bUseGravityOverride=true
    gravityScale=0.35
    alphaBoolBlend {blendOption=Linear}
    boundBone {boneName=dyn_burger_layer_01}
    componentAppliedLinearAccClamp=<5000 5000 5000>
    componentLinearAccScale=<0.5 0.5 0.5>
    componentLinearVelScale=<0.5 0.5 0.5>
    gravityOverride {z=980}
    physicsBodyDefinitions [{
      boundBone {boneName=dyn_burger_layer_01}
      boxExtents=<3 30 30>
      constraintSetup {angularLimitsMax=<0 0 4>, angularLimitsMin=<0 0 -4>, angularTarget=<1 0 0>}
      localJointOffset=<2 0 0>
     }]
   }
   n2 = Anim Dynamics - Bone: dyn_burger_layer_02!(ComponentPose=@n1, Alpha=1, GravityOverride:Vector="0,0,980", ExternalForce="0,0,0") -> (Pose) {
    angularDampingOverride=0.9999
    angularSpringConstant=7000
    bAngularSpring=true
    bOverrideAngularDamping=true
    bUseGravityOverride=true
    gravityScale=0.35
    alphaBoolBlend {blendOption=Linear}
    boundBone {boneName=dyn_burger_layer_02}
    componentAppliedLinearAccClamp=<5000 5000 5000>
    componentLinearAccScale=<0.5 0.5 0.5>
    componentLinearVelScale=<0.5 0.5 0.5>
    gravityOverride {z=980}
    physicsBodyDefinitions [{
      boundBone {boneName=dyn_burger_layer_02}
      boxExtents=<3 30 30>
      constraintSetup {angularLimitsMax=<0 0 4>, angularLimitsMin=<0 0 -4>, angularTarget=<1 0 0>}
      localJointOffset=<2 0 0>
     }]
   }
   n3 = Anim Dynamics - Bone: dyn_burger_layer_03!(ComponentPose=@n2, Alpha=1, GravityOverride:Vector="0,0,980", ExternalForce="0,0,0") -> (Pose) {
    angularDampingOverride=0.9999
    angularSpringConstant=7000
    bAngularSpring=true
    bOverrideAngularDamping=true
    bUseGravityOverride=true
    gravityScale=0.35
    alphaBoolBlend {blendOption=Linear}
    boundBone {boneName=dyn_burger_layer_03}
    componentAppliedLinearAccClamp=<5000 5000 5000>
    componentLinearAccScale=<0.5 0.5 0.5>
    componentLinearVelScale=<0.5 0.5 0.5>
    gravityOverride {z=980}
    physicsBodyDefinitions [{
      boundBone {boneName=dyn_burger_layer_03}
      boxExtents=<3 30 30>
      constraintSetup {angularLimitsMax=<0 0 4>, angularLimitsMin=<0 0 -4>, angularTarget=<1 0 0>}
      localJointOffset=<2 0 0>
     }]
   }
   n4 = Anim Dynamics - Bone: dyn_burger_layer_04!(ComponentPose=@n3, Alpha=1, GravityOverride:Vector="0,0,980", ExternalForce="0,0,0") -> (Pose) {
    angularDampingOverride=0.9999
    angularSpringConstant=7000
    bAngularSpring=true
    bOverrideAngularDamping=true
    bUseGravityOverride=true
    gravityScale=0.35
    alphaBoolBlend {blendOption=Linear}
    boundBone {boneName=dyn_burger_layer_04}
    componentAppliedLinearAccClamp=<5000 5000 5000>
    componentLinearAccScale=<0.5 0.5 0.5>
    componentLinearVelScale=<0.5 0.5 0.5>
    gravityOverride {z=980}
    physicsBodyDefinitions [{
      boundBone {boneName=dyn_burger_layer_04}
      boxExtents=<3 30 30>
      constraintSetup {angularLimitsMax=<0 0 4>, angularLimitsMin=<0 0 -4>, angularTarget=<1 0 0>}
      localJointOffset=<2 0 0>
     }]
   }
   n5 = Anim Dynamics - Bone: dyn_burger_layer_05!(ComponentPose=@n4, Alpha=1, GravityOverride:Vector="0,0,980", ExternalForce="0,0,0") -> (Pose) {
    angularDampingOverride=0.9999
    angularSpringConstant=7000
    bAngularSpring=true
    bOverrideAngularDamping=true
    bUseGravityOverride=true
    gravityScale=0.35
    alphaBoolBlend {blendOption=Linear}
    boundBone {boneName=dyn_burger_layer_05}
    componentAppliedLinearAccClamp=<5000 5000 5000>
    componentLinearAccScale=<0.5 0.5 0.5>
    componentLinearVelScale=<0.5 0.5 0.5>
    gravityOverride {z=980}
    physicsBodyDefinitions [{
      boundBone {boneName=dyn_burger_layer_05}
      boxExtents=<3 30 30>
      constraintSetup {angularLimitsMax=<0 0 4>, angularLimitsMin=<0 0 -4>, angularTarget=<1 0 0>}
      localJointOffset=<2 0 0>
     }]
   }
   n6 = Anim Dynamics - Bone: dyn_burger_layer_06!(ComponentPose=@n5, Alpha=1, GravityOverride:Vector="0,0,980", ExternalForce="0,0,0") -> (Pose) {
    angularDampingOverride=0.9999
    angularSpringConstant=7000
    bAngularSpring=true
    bOverrideAngularDamping=true
    bUseGravityOverride=true
    gravityScale=0.35
    alphaBoolBlend {blendOption=Linear}
    boundBone {boneName=dyn_burger_layer_06}
    componentAppliedLinearAccClamp=<5000 5000 5000>
    componentLinearAccScale=<0.5 0.5 0.5>
    componentLinearVelScale=<0.5 0.5 0.5>
    gravityOverride {z=980}
    physicsBodyDefinitions [{
      boundBone {boneName=dyn_burger_layer_06}
      boxExtents=<3 30 30>
      constraintSetup {angularLimitsMax=<0 0 4>, angularLimitsMin=<0 0 -4>, angularTarget=<1 0 0>}
      localJointOffset=<2 0 0>
     }]
   }
   n7 = Anim Dynamics - Bone: dyn_burger_layer_07!(ComponentPose=@n6, Alpha=1, GravityOverride:Vector="0,0,980", ExternalForce="0,0,0") -> (Pose) {
    angularDampingOverride=0.9999
    angularSpringConstant=7000
    bAngularSpring=true
    bOverrideAngularDamping=true
    bUseGravityOverride=true
    gravityScale=0.35
    alphaBoolBlend {blendOption=Linear}
    boundBone {boneName=dyn_burger_layer_07}
    componentAppliedLinearAccClamp=<5000 5000 5000>
    componentLinearAccScale=<0.5 0.5 0.5>
    componentLinearVelScale=<0.5 0.5 0.5>
    gravityOverride {z=980}
    physicsBodyDefinitions [{
      boundBone {boneName=dyn_burger_layer_07}
      boxExtents=<3 30 30>
      constraintSetup {angularLimitsMax=<0 0 4>, angularLimitsMin=<0 0 -4>, angularTarget=<1 0 0>}
      localJointOffset=<2 0 0>
     }]
   }
   n8 = Anim Dynamics - Bone: dyn_burger_layer_08!(ComponentPose=@n7, Alpha=1, GravityOverride:Vector="0,0,980", ExternalForce="0,0,0") -> (Pose) {
    angularDampingOverride=0.9999
    angularSpringConstant=7000
    bAngularSpring=true
    bOverrideAngularDamping=true
    bUseGravityOverride=true
    gravityScale=0.35
    alphaBoolBlend {blendOption=Linear}
    boundBone {boneName=dyn_burger_layer_08}
    componentAppliedLinearAccClamp=<5000 5000 5000>
    componentLinearAccScale=<0.5 0.5 0.5>
    componentLinearVelScale=<0.5 0.5 0.5>
    gravityOverride {z=980}
    physicsBodyDefinitions [{
      boundBone {boneName=dyn_burger_layer_08}
      boxExtents=<3 30 30>
      constraintSetup {angularLimitsMax=<0 0 4>, angularLimitsMin=<0 0 -4>, angularTarget=<1 0 0>}
      localJointOffset=<2 0 0>
     }]
   }
   Anim Dynamics - Bone: dyn_burger_layer_09!(ComponentPose=@n8, Alpha=1, GravityOverride:Vector="0,0,980", ExternalForce="0,0,0") -> (Pose) {
    angularDampingOverride=0.9999
    angularSpringConstant=7000
    bAngularSpring=true
    bOverrideAngularDamping=true
    bUseGravityOverride=true
    gravityScale=0.35
    alphaBoolBlend {blendOption=Linear}
    boundBone {boneName=dyn_burger_layer_09}
    componentAppliedLinearAccClamp=<5000 5000 5000>
    componentLinearAccScale=<0.5 0.5 0.5>
    componentLinearVelScale=<0.5 0.5 0.5>
    gravityOverride {z=980}
    physicsBodyDefinitions [{
      boundBone {boneName=dyn_burger_layer_09}
      boxExtents=<3 30 30>
      constraintSetup {angularLimitsMax=<0 0 4>, angularLimitsMin=<0 0 -4>, angularTarget=<1 0 0>}
      localJointOffset=<2 0 0>
     }]
   }
   AnimGraphNode_StateMachine_1 = OnKillStateMachine() {is=title AnimGraphNode_StateMachine}
   Output Pose(Result=Blend Poses by bool(
     BlendPose_0=@AnimGraphNode_StateMachine_1
     BlendPose_1=Sequence Player 'FluteLamp_Stage_00_Idle'() {bLoopAnimation=false, sequence=/BRCosmetics/Accessories/FORT_Backpacks/Backpack_M_MED_FluteLamp/Animation/FluteLamp_Stage_00_Idle}
     bActiveValue=Get CosmeticVariantReactivity() -> (CosmeticVariantReactivity:bool)
    ) {blendType=Linear}) {layerGroup=None}
   AnimGraphNode_StateMachine_0 = OnKillStateMachine_1() {is=title AnimGraphNode_StateMachine}
  ]}
 OnKillStateMachine_1 {
  entry=Idle
  path=AnimGraph.AnimGraphNode_StateMachine_0.OnKillStateMachine_1
  ids=id
  nodes [
   AnimStateNode_0 = Idle()
   AnimStateNode_1 = OnKill() {is=title AnimStateNode}
   AnimStateTransitionNode_0 = AnimStateTransitionNode() {blendMode=Linear, crossfadeDuration=0}
   AnimStateTransitionNode_1 = AnimStateTransitionNode() {bAutomaticRuleBasedOnSequencePlayerInState=true, blendMode=Linear, crossfadeDuration=0}
  ]
  links [
   {
    dst=OnKill
    rule=AnimGraph.AnimGraphNode_StateMachine_0.OnKillStateMachine_1.AnimStateTransitionNode_0.Transition
    src=Idle
    via=AnimStateTransitionNode_0
   }
   {
    dst=Idle
    rule=AnimGraph.AnimGraphNode_StateMachine_0.OnKillStateMachine_1.AnimStateTransitionNode_1.Transition
    src=OnKill
    via=AnimStateTransitionNode_1
   }
  ]
 }
 Idle {
  path=AnimGraph.AnimGraphNode_StateMachine_0.OnKillStateMachine_1.AnimStateNode_0.Idle
  nodes [
   n1 = Anim Dynamics - Bone: dyn_burger_layer_01(
    ComponentPose=Local To Component(LocalPose=Sequence Player 'FluteLamp_Stage_00_Idle'() {bLoopAnimation=false, sequence=&FluteLamp_Stage_00_Idle})
    Alpha=1
    GravityOverride="0,0,980"
    ExternalForce="0,0,0"
   ) {
    angularDampingOverride=0.9999
    angularSpringConstant=7000
    bAngularSpring=true
    bOverrideAngularDamping=true
    bUseGravityOverride=true
    gravityScale=0.35
    alphaBoolBlend {blendOption=Linear}
    boundBone {boneName=dyn_burger_layer_01}
    componentAppliedLinearAccClamp=<5000 5000 5000>
    componentLinearAccScale=<0.5 0.5 0.5>
    componentLinearVelScale=<0.5 0.5 0.5>
    gravityOverride {z=980}
    physicsBodyDefinitions [{
      boundBone {boneName=dyn_burger_layer_01}
      boxExtents=<3 30 30>
      constraintSetup {angularLimitsMax=<0 0 7>, angularLimitsMin=<0 0 -7>, angularTarget=<1 0 0>}
      localJointOffset=<3 0 0>
     }]
   }
   n2 = Anim Dynamics - Bone: dyn_burger_layer_09(ComponentPose=@n1, Alpha=1, GravityOverride="0,0,980", ExternalForce="0,0,0") {
    angularDampingOverride=0.9999
    angularSpringConstant=7000
    bAngularSpring=true
    bOverrideAngularDamping=true
    bUseGravityOverride=true
    gravityScale=0.35
    alphaBoolBlend {blendOption=Linear}
    boundBone {boneName=dyn_burger_layer_09}
    componentAppliedLinearAccClamp=<5000 5000 5000>
    componentLinearAccScale=<0.5 0.5 0.5>
    componentLinearVelScale=<0.5 0.5 0.5>
    gravityOverride {z=980}
    physicsBodyDefinitions [{
      boundBone {boneName=dyn_burger_layer_09}
      boxExtents=<3 30 30>
      constraintSetup {angularLimitsMax=<0 0 6>, angularLimitsMin=<0 0 -6>, angularTarget=<1 0 0>}
      localJointOffset=<2 0 0>
     }]
   }
   n3 = RigidBody!(
    ComponentPose=Local To Component(LocalPose=Sequence Player 'FluteLamp_Stage_04_Idle'() {bLoopAnimation=false, sequence=/BRCosmetics/Accessories/FORT_Backpacks/Backpack_M_MED_FluteLamp/Animation/FluteLamp_Stage_04_Idle})
    Alpha
    OverrideWorldGravity:Vector="0,0,980"
    ExternalForce
   ) -> (Pose) {
    bOverrideWorldGravity=true
    overridePhysicsAsset=&/M_MED_FluteLamp_Pack_Physics
    simulationSpace=BaseBoneSpace
    alphaBoolBlend {blendOption=Linear}
    baseBoneRef {boneName=pack}
    componentAppliedLinearAccClamp {y=0}
    overrideWorldGravity {z=980}
    simSpaceSettings {
     dampingAlpha=0
     maxAngularAcceleration=3
     maxAngularVelocity=1
     maxLinearAcceleration=500
     maxLinearVelocity=200
     velocityScaleZ=0
     worldAlpha=0.5
     externalLinearDragV {x=3, z=1}
    }
   }
   n4 = Anim Dynamics - Bone: dyn_burger_layer_01(
    ComponentPose=Local To Component(LocalPose=Sequence Player 'FluteLamp_Stage_03_Idle'() {bLoopAnimation=false, sequence=/BRCosmetics/Accessories/FORT_Backpacks/Backpack_M_MED_FluteLamp/Animation/FluteLamp_Stage_03_Idle})
    Alpha=1
    GravityOverride="0,0,980"
    ExternalForce="0,0,0"
   ) {
    angularDampingOverride=1
    angularSpringConstant=1000
    bAngularSpring=true
    bOverrideAngularDamping=true
    bUseGravityOverride=true
    gravityScale=0.35
    alphaBoolBlend {blendOption=Linear}
    boundBone {boneName=dyn_burger_layer_01}
    componentAppliedLinearAccClamp=<5000 5000 5000>
    componentLinearAccScale=<0.5 0.5 0.5>
    componentLinearVelScale=<0.5 0.5 0.5>
    gravityOverride {z=980}
    physicsBodyDefinitions [{
      boundBone {boneName=dyn_burger_layer_01}
      boxExtents=<3 30 30>
      constraintSetup {angularLimitsMax=<0 0 1>, angularLimitsMin=<0 0 -1>, angularTarget=<1 0 0>}
      localJointOffset=<2 0 0>
     }]
   }
   n5 = Anim Dynamics - Bone: dyn_burger_layer_04(ComponentPose=@n4, Alpha=1, GravityOverride="0,0,980", ExternalForce="0,0,0") {
    angularDampingOverride=1
    angularSpringConstant=3000
    bAngularSpring=true
    bOverrideAngularDamping=true
    bUseGravityOverride=true
    gravityScale=0.35
    alphaBoolBlend {blendOption=Linear}
    boundBone {boneName=dyn_burger_layer_04}
    componentAppliedLinearAccClamp=<5000 5000 5000>
    componentLinearAccScale=<0.5 0.5 0.5>
    componentLinearVelScale=<0.5 0.5 0.5>
    gravityOverride {z=980}
    physicsBodyDefinitions [{
      boundBone {boneName=dyn_burger_layer_04}
      boxExtents=<3 30 30>
      constraintSetup {angularLimitsMax=<0 0 2>, angularLimitsMin=<0 0 -2>, angularTarget=<1 0 0>}
      localJointOffset=<2 0 0>
     }]
   }
   n6 = Anim Dynamics - Bone: dyn_burger_layer_05(ComponentPose=@n5, Alpha=1, GravityOverride="0,0,980", ExternalForce="0,0,0") {
    angularDampingOverride=1
    angularSpringConstant=5000
    bAngularSpring=true
    bOverrideAngularDamping=true
    bUseGravityOverride=true
    gravityScale=0.35
    alphaBoolBlend {blendOption=Linear}
    boundBone {boneName=dyn_burger_layer_05}
    componentAppliedLinearAccClamp=<5000 5000 5000>
    componentLinearAccScale=<0.5 0.5 0.5>
    componentLinearVelScale=<0.5 0.5 0.5>
    gravityOverride {z=980}
    physicsBodyDefinitions [{
      boundBone {boneName=dyn_burger_layer_05}
      boxExtents=<3 30 30>
      constraintSetup {angularLimitsMax=<0 0 5>, angularLimitsMin=<0 0 -5>, angularTarget=<1 0 0>}
      localJointOffset=<2 0 0>
     }]
   }
   n7 = Anim Dynamics - Bone: dyn_burger_layer_06(ComponentPose=@n6, Alpha=1, GravityOverride="0,0,980", ExternalForce="0,0,0") {
    angularDampingOverride=1
    angularSpringConstant=3000
    bAngularSpring=true
    bOverrideAngularDamping=true
    bUseGravityOverride=true
    gravityScale=0.35
    alphaBoolBlend {blendOption=Linear}
    boundBone {boneName=dyn_burger_layer_06}
    componentAppliedLinearAccClamp=<5000 5000 5000>
    componentLinearAccScale=<0.5 0.5 0.5>
    componentLinearVelScale=<0.5 0.5 0.5>
    gravityOverride {z=980}
    physicsBodyDefinitions [{
      boundBone {boneName=dyn_burger_layer_06}
      boxExtents=<5 30 30>
      constraintSetup {angularLimitsMax=<0 0 5>, angularLimitsMin=<0 0 -5>, angularTarget=<1 0 0>}
      localJointOffset=<2 0 0>
     }]
   }
   n8 = Anim Dynamics - Bone: dyn_burger_layer_07(ComponentPose=@n7, Alpha=1, GravityOverride="0,0,980", ExternalForce="0,0,0") {
    angularDampingOverride=1
    angularSpringConstant=3000
    bAngularSpring=true
    bOverrideAngularDamping=true
    bUseGravityOverride=true
    gravityScale=0.35
    alphaBoolBlend {blendOption=Linear}
    boundBone {boneName=dyn_burger_layer_07}
    componentAppliedLinearAccClamp=<5000 5000 5000>
    componentLinearAccScale=<0.5 0.5 0.5>
    componentLinearVelScale=<0.5 0.5 0.5>
    gravityOverride {z=980}
    physicsBodyDefinitions [{
      boundBone {boneName=dyn_burger_layer_07}
      boxExtents=<10 30 30>
      constraintSetup {angularLimitsMax=<0 0 5>, angularLimitsMin=<0 0 -5>, angularTarget=<1 0 0>}
      localJointOffset=<2 0 0>
     }]
   }
   n9 = Anim Dynamics - Bone: dyn_burger_layer_08(ComponentPose=@n8, Alpha=1, GravityOverride="0,0,980", ExternalForce="0,0,0") {
    angularDampingOverride=1
    angularSpringConstant=7000
    bAngularSpring=true
    bOverrideAngularDamping=true
    bUseGravityOverride=true
    gravityScale=0.35
    alphaBoolBlend {blendOption=Linear}
    boundBone {boneName=dyn_burger_layer_08}
    componentAppliedLinearAccClamp=<5000 5000 5000>
    componentLinearAccScale=<0.5 0.5 0.5>
    componentLinearVelScale=<0.5 0.5 0.5>
    gravityOverride {z=980}
    physicsBodyDefinitions [{
      boundBone {boneName=dyn_burger_layer_08}
      boxExtents=<3 30 30>
      constraintSetup {angularLimitsMax=<0 0 8>, angularLimitsMin=<0 0 -8>, angularTarget=<1 0 0>}
      localJointOffset=<2 0 0>
     }]
   }
   n10 = Anim Dynamics - Bone: dyn_burger_layer_09(ComponentPose=@n9, Alpha=1, GravityOverride="0,0,980", ExternalForce="0,0,0") {
    angularDampingOverride=1
    angularSpringConstant=7000
    bAngularSpring=true
    bOverrideAngularDamping=true
    bUseGravityOverride=true
    gravityScale=0.35
    alphaBoolBlend {blendOption=Linear}
    boundBone {boneName=dyn_burger_layer_09}
    componentAppliedLinearAccClamp=<5000 5000 5000>
    componentLinearAccScale=<0.5 0.5 0.5>
    componentLinearVelScale=<0.5 0.5 0.5>
    gravityOverride {z=980}
    physicsBodyDefinitions [{
      boundBone {boneName=dyn_burger_layer_09}
      boxExtents=<3 30 30>
      constraintSetup {angularLimitsMax=<0 0 8>, angularLimitsMin=<0 0 -8>, angularTarget=<1 0 0>}
      localJointOffset=<2 0 0>
     }]
   }
   n11 = Anim Dynamics - Bone: dyn_burger_layer_01(
    ComponentPose=Local To Component(LocalPose=Sequence Player 'FluteLamp_Stage_01_Idle'() {bLoopAnimation=false, sequence=/BRCosmetics/Accessories/FORT_Backpacks/Backpack_M_MED_FluteLamp/Animation/FluteLamp_Stage_01_Idle})
    Alpha=1
    GravityOverride="0,0,980"
    ExternalForce="0,0,0"
   ) {
    angularDampingOverride=1
    angularSpringConstant=7000
    bAngularSpring=true
    bOverrideAngularDamping=true
    bUseGravityOverride=true
    gravityScale=0.35
    alphaBoolBlend {blendOption=Linear}
    boundBone {boneName=dyn_burger_layer_01}
    componentAppliedLinearAccClamp=<5000 5000 5000>
    componentLinearAccScale=<0.5 0.5 0.5>
    componentLinearVelScale=<0.5 0.5 0.5>
    gravityOverride {z=980}
    physicsBodyDefinitions [{
      boundBone {boneName=dyn_burger_layer_01}
      boxExtents=<3 30 30>
      constraintSetup {angularLimitsMax=<0 0 1>, angularLimitsMin=<0 0 -1>, angularTarget=<1 0 0>}
      localJointOffset=<2 0 0>
     }]
   }
   n12 = Anim Dynamics - Bone: dyn_burger_layer_08(ComponentPose=@n11, Alpha=1, GravityOverride="0,0,980", ExternalForce="0,0,0") {
    angularDampingOverride=0.9999
    angularSpringConstant=2000
    bAngularSpring=true
    bOverrideAngularDamping=true
    bUseGravityOverride=true
    gravityScale=0.35
    alphaBoolBlend {blendOption=Linear}
    boundBone {boneName=dyn_burger_layer_08}
    componentAppliedLinearAccClamp=<5000 5000 5000>
    componentLinearAccScale=<0.5 0.5 0.5>
    componentLinearVelScale=<0.5 0.5 0.5>
    gravityOverride {z=980}
    physicsBodyDefinitions [{
      boundBone {boneName=dyn_burger_layer_08}
      boxExtents=<11 30 30>
      constraintSetup {angularLimitsMax=<0 0 6>, angularLimitsMin=<0 0 -6>, angularTarget=<1 0 0>}
      localJointOffset=<2 0 0>
     }]
   }
   n13 = Anim Dynamics - Bone: dyn_burger_layer_09(ComponentPose=@n12, Alpha=1, GravityOverride="0,0,980", ExternalForce="0,0,0") {
    angularDampingOverride=1
    angularSpringConstant=5000
    bAngularSpring=true
    bOverrideAngularDamping=true
    bUseGravityOverride=true
    gravityScale=0.35
    alphaBoolBlend {blendOption=Linear}
    boundBone {boneName=dyn_burger_layer_09}
    componentAppliedLinearAccClamp=<5000 5000 5000>
    componentLinearAccScale=<0.5 0.5 0.5>
    componentLinearVelScale=<0.5 0.5 0.5>
    gravityOverride {z=980}
    physicsBodyDefinitions [{
      boundBone {boneName=dyn_burger_layer_09}
      boxExtents=<3 30 30>
      constraintSetup {angularLimitsMax=<0 0 6>, angularLimitsMin=<0 0 -6>, angularTarget=<1 0 0>}
      localJointOffset=<2 0 0>
     }]
   }
   n14 = Anim Dynamics - Bone: dyn_burger_layer_01(
    ComponentPose=Local To Component(LocalPose=Sequence Player 'FluteLamp_Stage_02_Idle'() {bLoopAnimation=false, sequence=/BRCosmetics/Accessories/FORT_Backpacks/Backpack_M_MED_FluteLamp/Animation/FluteLamp_Stage_02_Idle})
    Alpha=1
    GravityOverride="0,0,980"
    ExternalForce="0,0,0"
   ) {
    angularDampingOverride=1
    angularSpringConstant=1000
    bAngularSpring=true
    bOverrideAngularDamping=true
    bUseGravityOverride=true
    gravityScale=0.35
    alphaBoolBlend {blendOption=Linear}
    boundBone {boneName=dyn_burger_layer_01}
    componentAppliedLinearAccClamp=<5000 5000 5000>
    componentLinearAccScale=<0.5 0.5 0.5>
    componentLinearVelScale=<0.5 0.5 0.5>
    gravityOverride {z=980}
    physicsBodyDefinitions [{
      boundBone {boneName=dyn_burger_layer_01}
      boxExtents=<3 30 30>
      constraintSetup {angularLimitsMax=<0 0 1>, angularLimitsMin=<0 0 -1>, angularTarget=<1 0 0>}
      localJointOffset=<2 0 0>
     }]
   }
   n15 = Anim Dynamics - Bone: dyn_burger_layer_06(ComponentPose=@n14, Alpha=1, GravityOverride="0,0,980", ExternalForce="0,0,0") {
    angularDampingOverride=1
    angularSpringConstant=7000
    bAngularSpring=true
    bOverrideAngularDamping=true
    bUseGravityOverride=true
    gravityScale=0.35
    alphaBoolBlend {blendOption=Linear}
    boundBone {boneName=dyn_burger_layer_06}
    componentAppliedLinearAccClamp=<5000 5000 5000>
    componentLinearAccScale=<0.5 0.5 0.5>
    componentLinearVelScale=<0.5 0.5 0.5>
    gravityOverride {z=980}
    physicsBodyDefinitions [{
      boundBone {boneName=dyn_burger_layer_06}
      boxExtents=<3 30 30>
      constraintSetup {angularLimitsMax=<0 0 3>, angularLimitsMin=<0 0 -3>, angularTarget=<1 0 0>}
      localJointOffset=<2 0 0>
     }]
   }
   n16 = Anim Dynamics - Bone: dyn_burger_layer_07(ComponentPose=@n15, Alpha=1, GravityOverride="0,0,980", ExternalForce="0,0,0") {
    angularDampingOverride=1
    angularSpringConstant=7000
    bAngularSpring=true
    bOverrideAngularDamping=true
    bUseGravityOverride=true
    gravityScale=0.35
    alphaBoolBlend {blendOption=Linear}
    boundBone {boneName=dyn_burger_layer_07}
    componentAppliedLinearAccClamp=<5000 5000 5000>
    componentLinearAccScale=<0.5 0.5 0.5>
    componentLinearVelScale=<0.5 0.5 0.5>
    gravityOverride {z=980}
    physicsBodyDefinitions [{
      boundBone {boneName=dyn_burger_layer_07}
      boxExtents=<3 30 30>
      constraintSetup {angularLimitsMax=<0 0 3>, angularLimitsMin=<0 0 -3>, angularTarget=<1 0 0>}
      localJointOffset=<2 0 0>
     }]
   }
   n17 = Anim Dynamics - Bone: dyn_burger_layer_08(ComponentPose=@n16, Alpha=1, GravityOverride="0,0,980", ExternalForce="0,0,0") {
    angularDampingOverride=1
    angularSpringConstant=4000
    bAngularSpring=true
    bOverrideAngularDamping=true
    bUseGravityOverride=true
    gravityScale=0.35
    alphaBoolBlend {blendOption=Linear}
    boundBone {boneName=dyn_burger_layer_08}
    componentAppliedLinearAccClamp=<5000 5000 5000>
    componentLinearAccScale=<0.5 0.5 0.5>
    componentLinearVelScale=<0.5 0.5 0.5>
    gravityOverride {z=980}
    physicsBodyDefinitions [{
      boundBone {boneName=dyn_burger_layer_08}
      boxExtents=<3 30 30>
      constraintSetup {angularLimitsMax=<0 0 5>, angularLimitsMin=<0 0 -5>, angularTarget=<1 0 0>}
      localJointOffset=<2 0 0>
     }]
   }
   n18 = Anim Dynamics - Bone: dyn_burger_layer_09(ComponentPose=@n17, Alpha=1, GravityOverride="0,0,980", ExternalForce="0,0,0") {
    angularDampingOverride=1
    angularSpringConstant=7000
    bAngularSpring=true
    bOverrideAngularDamping=true
    bUseGravityOverride=true
    gravityScale=0.35
    alphaBoolBlend {blendOption=Linear}
    boundBone {boneName=dyn_burger_layer_09}
    componentAppliedLinearAccClamp=<5000 5000 5000>
    componentLinearAccScale=<0.5 0.5 0.5>
    componentLinearVelScale=<0.5 0.5 0.5>
    gravityOverride {z=980}
    physicsBodyDefinitions [{
      boundBone {boneName=dyn_burger_layer_09}
      boxExtents=<3 30 30>
      constraintSetup {angularLimitsMax=<0 0 7>, angularLimitsMin=<0 0 -7>, angularTarget=<1 0 0>}
      localJointOffset=<2 0 0>
     }]
   }
   n19 = Blend Poses by int!(
    BlendPose_0:Pose Link Structure=Component To Local(ComponentPose=@n2)
    BlendPose_1:Pose Link Structure=Component To Local(ComponentPose=@n13)
    BlendPose_2:Pose Link Structure=Component To Local(ComponentPose=@n18)
    BlendPose_3:Pose Link Structure=Component To Local(ComponentPose=@n10)
    BlendPose_4:Pose Link Structure=Component To Local(ComponentPose=@n3)
    BlendTime_0:float=0
    BlendTime_1:float=0
    BlendTime_2:float=0
    BlendTime_3:float=0
    BlendTime_4:float=0
    ActiveChildIndex:int=Get KillCounter() -> (KillCounter:int=0)
   ) -> (Pose:Pose Link Structure) {
    is=title AnimGraphNode_BlendListByInt
    blendType=Linear
    blendPose [
     {linkId=-1, sourceLinkId=-1}
     {linkId=-1, sourceLinkId=-1}
     {linkId=-1, sourceLinkId=-1}
     {linkId=-1, sourceLinkId=-1}
     {linkId=-1, sourceLinkId=-1}
    ]
    blendTime [0.1,0.1,0.1,0.1,0.1]
   }
   Output Animation Pose(Result=@n19)
  ]
 }
 OnKill {
  path=AnimGraph.AnimGraphNode_StateMachine_0.OnKillStateMachine_1.AnimStateNode_1.OnKill
  nodes [
   n1 = Sequence Player 'FluteLamp_Stage04Loop_Reactive'() {bLoopAnimation=false, sequence=/BRCosmetics/Accessories/FORT_Backpacks/Backpack_M_MED_FluteLamp/Animation/FluteLamp_Stage04Loop_Reactive}
   n2 = Sequence Player 'FluteLamp_Stage01_Reactive'() {bLoopAnimation=false, sequence=/BRCosmetics/Accessories/FORT_Backpacks/Backpack_M_MED_FluteLamp/Animation/FluteLamp_Stage01_Reactive}
   n3 = Sequence Player 'FluteLamp_Stage02_Reactive'() {bLoopAnimation=false, sequence=/BRCosmetics/Accessories/FORT_Backpacks/Backpack_M_MED_FluteLamp/Animation/FluteLamp_Stage02_Reactive}
   n4 = Sequence Player 'FluteLamp_Stage03_Reactive'() {bLoopAnimation=false, sequence=/BRCosmetics/Accessories/FORT_Backpacks/Backpack_M_MED_FluteLamp/Animation/FluteLamp_Stage03_Reactive}
   n5 = Sequence Player 'FluteLamp_Stage04_Reactive'() {bLoopAnimation=false, sequence=/BRCosmetics/Accessories/FORT_Backpacks/Backpack_M_MED_FluteLamp/Animation/FluteLamp_Stage04_Reactive}
   n6 = Blend Poses by int!(
    BlendPose_0
    BlendPose_1=@n2
    BlendPose_2=@n3
    BlendPose_3=@n4
    BlendPose_4=@n5
    BlendPose_5:Pose Link Structure=@n1
    BlendTime_0=0
    BlendTime_1=0
    BlendTime_2=0
    BlendTime_3=0
    BlendTime_4=0.1
    BlendTime_5:float=0.1
    ActiveChildIndex=Get KillCounter() -> (KillCounter=0)
   ) -> (Pose) {
    blendType=Linear
    blendPose [
     {linkId=-1, sourceLinkId=-1}
     {linkId=-1, sourceLinkId=-1}
     {linkId=-1, sourceLinkId=-1}
     {linkId=-1, sourceLinkId=-1}
     {linkId=-1, sourceLinkId=-1}
     {linkId=-1, sourceLinkId=-1}
    ]
    blendTime [0.1,0.1,0.1,0.1,0.1,0.1]
   }
   Output Animation Pose(Result=@n6)
  ]
 }
 Transition {
  path=AnimGraph.AnimGraphNode_StateMachine_0.OnKillStateMachine_1.AnimStateTransitionNode_0.Transition
  nodes [
   n1 = Get ActiveKill() -> (ActiveKill:bool)
   Result(bCanEnterTransition=@n1)
   AND Boolean(A=@n1, B="integer < integer"(A=Get KillCounter() -> (KillCounter=0), B=Get MaxKills() -> (MaxKills:int=0)))
  ]
 }
 Transition {path=AnimGraph.AnimGraphNode_StateMachine_0.OnKillStateMachine_1.AnimStateTransitionNode_1.Transition, nodes [Result()]}
 OnKillStateMachine {
  entry=Idle
  path=AnimGraph.AnimGraphNode_StateMachine_1.OnKillStateMachine
  ids=id
  nodes [
   AnimStateNode_0 = Idle()
   AnimStateNode_1 = OnKill()
   AnimStateTransitionNode_0 = AnimStateTransitionNode() {blendMode=Linear, crossfadeDuration=0}
   AnimStateTransitionNode_1 = AnimStateTransitionNode() {bAutomaticRuleBasedOnSequencePlayerInState=true}
  ]
  links [
   {
    dst=OnKill
    rule=AnimGraph.AnimGraphNode_StateMachine_1.OnKillStateMachine.AnimStateTransitionNode_0.Transition
    src=Idle
    via=AnimStateTransitionNode_0
   }
   {
    dst=Idle
    rule=AnimGraph.AnimGraphNode_StateMachine_1.OnKillStateMachine.AnimStateTransitionNode_1.Transition
    src=OnKill
    via=AnimStateTransitionNode_1
   }
  ]
 }
 Idle {
  path=AnimGraph.AnimGraphNode_StateMachine_1.OnKillStateMachine.AnimStateNode_0.Idle
  nodes [
   Anim Dynamics - Bone: dyn_burger_layer_01!(ComponentPose, AlphaCurveName:Name=DynamicsAlpha, GravityOverride="0,0,980", ExternalForce="0,0,0") -> (Pose) {
    alphaInputType=Curve
    angularDampingOverride=0.9999
    angularSpringConstant=7000
    bAngularSpring=true
    bOverrideAngularDamping=true
    bUseGravityOverride=true
    gravityScale=0.35
    alphaBoolBlend {blendOption=Linear}
    boundBone {boneName=dyn_burger_layer_01}
    componentAppliedLinearAccClamp=<5000 5000 5000>
    componentLinearAccScale=<0.5 0.5 0.5>
    componentLinearVelScale=<0.5 0.5 0.5>
    gravityOverride {z=980}
    physicsBodyDefinitions [{
      boundBone {boneName=dyn_burger_layer_01}
      boxExtents=<3 30 30>
      constraintSetup {angularLimitsMax=<0 0 7>, angularLimitsMin=<0 0 -7>, angularTarget=<1 0 0>}
      localJointOffset=<3 0 0>
     }]
   }
   n1 = Anim Dynamics - Bone: dyn_burger_layer_09!(
    ComponentPose=Local To Component(LocalPose=Sequence Player 'FluteLamp_Stage_00_Idle'() {bLoopAnimation=false, sequence=&FluteLamp_Stage_00_Idle})
    AlphaCurveName:Name=DynamicsAlpha
    GravityOverride="0,0,980"
    ExternalForce="0,0,0"
   ) -> (Pose) {
    alphaInputType=Curve
    angularDampingOverride=0.9999
    angularSpringConstant=7000
    bAngularSpring=true
    bOverrideAngularDamping=true
    bUseGravityOverride=true
    gravityScale=0.35
    alphaBoolBlend {blendOption=Linear}
    boundBone {boneName=dyn_burger_layer_09}
    componentAppliedLinearAccClamp=<5000 5000 5000>
    componentLinearAccScale=<0.5 0.5 0.5>
    componentLinearVelScale=<0.5 0.5 0.5>
    gravityOverride {z=980}
    physicsBodyDefinitions [{
      boundBone {boneName=dyn_burger_layer_09}
      boxExtents=<3 30 30>
      constraintSetup {angularLimitsMax=<0 0 6>, angularLimitsMin=<0 0 -6>, angularTarget=<1 0 0>}
      localJointOffset=<2 0 0>
     }]
   }
   n2 = RigidBody!(
    ComponentPose=Local To Component(LocalPose=Sequence Player 'FluteLamp_Stage_04_Idle'() {bLoopAnimation=false, sequence=&FluteLamp_Stage_04_Idle})
    AlphaCurveName:Name=DynamicsAlpha
    OverrideWorldGravity="0,0,980"
    ExternalForce
   ) -> (Pose) {
    alphaInputType=Curve
    bOverrideWorldGravity=true
    overridePhysicsAsset=&M_MED_FluteLamp_Pack_Physics
    simulationSpace=BaseBoneSpace
    alphaBoolBlend {blendOption=Linear}
    baseBoneRef {boneName=pack}
    componentAppliedLinearAccClamp {y=0}
    overrideWorldGravity {z=980}
    simSpaceSettings {
     dampingAlpha=0
     maxAngularAcceleration=3
     maxAngularVelocity=1
     maxLinearAcceleration=500
     maxLinearVelocity=200
     velocityScaleZ=0
     worldAlpha=0.5
     externalLinearDragV {x=3, z=1}
    }
   }
   n3 = Anim Dynamics - Bone: dyn_burger_layer_01(
    ComponentPose=Local To Component(LocalPose=Sequence Player 'FluteLamp_Stage_03_Idle'() {bLoopAnimation=false, sequence=&FluteLamp_Stage_03_Idle})
    AlphaCurveName=DynamicsAlpha
    GravityOverride="0,0,980"
    ExternalForce="0,0,0"
   ) {
    alphaInputType=Curve
    angularDampingOverride=1
    angularSpringConstant=1000
    bAngularSpring=true
    bOverrideAngularDamping=true
    bUseGravityOverride=true
    gravityScale=0.35
    alphaBoolBlend {blendOption=Linear}
    boundBone {boneName=dyn_burger_layer_01}
    componentAppliedLinearAccClamp=<5000 5000 5000>
    componentLinearAccScale=<0.5 0.5 0.5>
    componentLinearVelScale=<0.5 0.5 0.5>
    gravityOverride {z=980}
    physicsBodyDefinitions [{
      boundBone {boneName=dyn_burger_layer_01}
      boxExtents=<3 30 30>
      constraintSetup {angularLimitsMax=<0 0 1>, angularLimitsMin=<0 0 -1>, angularTarget=<1 0 0>}
      localJointOffset=<2 0 0>
     }]
   }
   n4 = Anim Dynamics - Bone: dyn_burger_layer_04!(ComponentPose=@n3, AlphaCurveName:Name=DynamicsAlpha, GravityOverride="0,0,980", ExternalForce="0,0,0") -> (Pose) {
    alphaInputType=Curve
    angularDampingOverride=1
    angularSpringConstant=3000
    bAngularSpring=true
    bOverrideAngularDamping=true
    bUseGravityOverride=true
    gravityScale=0.35
    alphaBoolBlend {blendOption=Linear}
    boundBone {boneName=dyn_burger_layer_04}
    componentAppliedLinearAccClamp=<5000 5000 5000>
    componentLinearAccScale=<0.5 0.5 0.5>
    componentLinearVelScale=<0.5 0.5 0.5>
    gravityOverride {z=980}
    physicsBodyDefinitions [{
      boundBone {boneName=dyn_burger_layer_04}
      boxExtents=<3 30 30>
      constraintSetup {angularLimitsMax=<0 0 2>, angularLimitsMin=<0 0 -2>, angularTarget=<1 0 0>}
      localJointOffset=<2 0 0>
     }]
   }
   n5 = Anim Dynamics - Bone: dyn_burger_layer_05!(ComponentPose=@n4, AlphaCurveName:Name=DynamicsAlpha, GravityOverride="0,0,980", ExternalForce="0,0,0") -> (Pose) {
    alphaInputType=Curve
    angularDampingOverride=1
    angularSpringConstant=5000
    bAngularSpring=true
    bOverrideAngularDamping=true
    bUseGravityOverride=true
    gravityScale=0.35
    alphaBoolBlend {blendOption=Linear}
    boundBone {boneName=dyn_burger_layer_05}
    componentAppliedLinearAccClamp=<5000 5000 5000>
    componentLinearAccScale=<0.5 0.5 0.5>
    componentLinearVelScale=<0.5 0.5 0.5>
    gravityOverride {z=980}
    physicsBodyDefinitions [{
      boundBone {boneName=dyn_burger_layer_05}
      boxExtents=<3 30 30>
      constraintSetup {angularLimitsMax=<0 0 5>, angularLimitsMin=<0 0 -5>, angularTarget=<1 0 0>}
      localJointOffset=<2 0 0>
     }]
   }
   n6 = Anim Dynamics - Bone: dyn_burger_layer_06!(ComponentPose=@n5, AlphaCurveName:Name=DynamicsAlpha, GravityOverride="0,0,980", ExternalForce="0,0,0") -> (Pose) {
    alphaInputType=Curve
    angularDampingOverride=1
    angularSpringConstant=3000
    bAngularSpring=true
    bOverrideAngularDamping=true
    bUseGravityOverride=true
    gravityScale=0.35
    alphaBoolBlend {blendOption=Linear}
    boundBone {boneName=dyn_burger_layer_06}
    componentAppliedLinearAccClamp=<5000 5000 5000>
    componentLinearAccScale=<0.5 0.5 0.5>
    componentLinearVelScale=<0.5 0.5 0.5>
    gravityOverride {z=980}
    physicsBodyDefinitions [{
      boundBone {boneName=dyn_burger_layer_06}
      boxExtents=<5 30 30>
      constraintSetup {angularLimitsMax=<0 0 5>, angularLimitsMin=<0 0 -5>, angularTarget=<1 0 0>}
      localJointOffset=<2 0 0>
     }]
   }
   n7 = Anim Dynamics - Bone: dyn_burger_layer_07!(ComponentPose=@n6, AlphaCurveName:Name=DynamicsAlpha, GravityOverride="0,0,980", ExternalForce="0,0,0") -> (Pose) {
    alphaInputType=Curve
    angularDampingOverride=1
    angularSpringConstant=3000
    bAngularSpring=true
    bOverrideAngularDamping=true
    bUseGravityOverride=true
    gravityScale=0.35
    alphaBoolBlend {blendOption=Linear}
    boundBone {boneName=dyn_burger_layer_07}
    componentAppliedLinearAccClamp=<5000 5000 5000>
    componentLinearAccScale=<0.5 0.5 0.5>
    componentLinearVelScale=<0.5 0.5 0.5>
    gravityOverride {z=980}
    physicsBodyDefinitions [{
      boundBone {boneName=dyn_burger_layer_07}
      boxExtents=<10 30 30>
      constraintSetup {angularLimitsMax=<0 0 5>, angularLimitsMin=<0 0 -5>, angularTarget=<1 0 0>}
      localJointOffset=<2 0 0>
     }]
   }
   n8 = Anim Dynamics - Bone: dyn_burger_layer_08!(ComponentPose=@n7, AlphaCurveName:Name=DynamicsAlpha, GravityOverride="0,0,980", ExternalForce="0,0,0") -> (Pose) {
    alphaInputType=Curve
    angularDampingOverride=1
    angularSpringConstant=7000
    bAngularSpring=true
    bOverrideAngularDamping=true
    bUseGravityOverride=true
    gravityScale=0.35
    alphaBoolBlend {blendOption=Linear}
    boundBone {boneName=dyn_burger_layer_08}
    componentAppliedLinearAccClamp=<5000 5000 5000>
    componentLinearAccScale=<0.5 0.5 0.5>
    componentLinearVelScale=<0.5 0.5 0.5>
    gravityOverride {z=980}
    physicsBodyDefinitions [{
      boundBone {boneName=dyn_burger_layer_08}
      boxExtents=<3 30 30>
      constraintSetup {angularLimitsMax=<0 0 8>, angularLimitsMin=<0 0 -8>, angularTarget=<1 0 0>}
      localJointOffset=<2 0 0>
     }]
   }
   n9 = Anim Dynamics - Bone: dyn_burger_layer_09(ComponentPose=@n8, AlphaCurveName=DynamicsAlpha, GravityOverride="0,0,980", ExternalForce="0,0,0") {
    alphaInputType=Curve
    angularDampingOverride=1
    angularSpringConstant=7000
    bAngularSpring=true
    bOverrideAngularDamping=true
    bUseGravityOverride=true
    gravityScale=0.35
    alphaBoolBlend {blendOption=Linear}
    boundBone {boneName=dyn_burger_layer_09}
    componentAppliedLinearAccClamp=<5000 5000 5000>
    componentLinearAccScale=<0.5 0.5 0.5>
    componentLinearVelScale=<0.5 0.5 0.5>
    gravityOverride {z=980}
    physicsBodyDefinitions [{
      boundBone {boneName=dyn_burger_layer_09}
      boxExtents=<3 30 30>
      constraintSetup {angularLimitsMax=<0 0 8>, angularLimitsMin=<0 0 -8>, angularTarget=<1 0 0>}
      localJointOffset=<2 0 0>
     }]
   }
   n10 = Anim Dynamics - Bone: dyn_burger_layer_01(
    ComponentPose=Local To Component(LocalPose=Sequence Player 'FluteLamp_Stage_01_Idle'() {bLoopAnimation=false, sequence=&FluteLamp_Stage_01_Idle})
    AlphaCurveName=DynamicsAlpha
    GravityOverride="0,0,980"
    ExternalForce="0,0,0"
   ) {
    alphaInputType=Curve
    angularDampingOverride=1
    angularSpringConstant=7000
    bAngularSpring=true
    bOverrideAngularDamping=true
    bUseGravityOverride=true
    gravityScale=0.35
    alphaBoolBlend {blendOption=Linear}
    boundBone {boneName=dyn_burger_layer_01}
    componentAppliedLinearAccClamp=<5000 5000 5000>
    componentLinearAccScale=<0.5 0.5 0.5>
    componentLinearVelScale=<0.5 0.5 0.5>
    gravityOverride {z=980}
    physicsBodyDefinitions [{
      boundBone {boneName=dyn_burger_layer_01}
      boxExtents=<3 30 30>
      constraintSetup {angularLimitsMax=<0 0 1>, angularLimitsMin=<0 0 -1>, angularTarget=<1 0 0>}
      localJointOffset=<2 0 0>
     }]
   }
   n11 = Anim Dynamics - Bone: dyn_burger_layer_08(ComponentPose=@n10, AlphaCurveName=DynamicsAlpha, GravityOverride="0,0,980", ExternalForce="0,0,0") {
    alphaInputType=Curve
    angularDampingOverride=0.9999
    angularSpringConstant=2000
    bAngularSpring=true
    bOverrideAngularDamping=true
    bUseGravityOverride=true
    gravityScale=0.35
    alphaBoolBlend {blendOption=Linear}
    boundBone {boneName=dyn_burger_layer_08}
    componentAppliedLinearAccClamp=<5000 5000 5000>
    componentLinearAccScale=<0.5 0.5 0.5>
    componentLinearVelScale=<0.5 0.5 0.5>
    gravityOverride {z=980}
    physicsBodyDefinitions [{
      boundBone {boneName=dyn_burger_layer_08}
      boxExtents=<11 30 30>
      constraintSetup {angularLimitsMax=<0 0 6>, angularLimitsMin=<0 0 -6>, angularTarget=<1 0 0>}
      localJointOffset=<2 0 0>
     }]
   }
   n12 = Anim Dynamics - Bone: dyn_burger_layer_09(ComponentPose=@n11, AlphaCurveName=DynamicsAlpha, GravityOverride="0,0,980", ExternalForce="0,0,0") {
    alphaInputType=Curve
    angularDampingOverride=1
    angularSpringConstant=5000
    bAngularSpring=true
    bOverrideAngularDamping=true
    bUseGravityOverride=true
    gravityScale=0.35
    alphaBoolBlend {blendOption=Linear}
    boundBone {boneName=dyn_burger_layer_09}
    componentAppliedLinearAccClamp=<5000 5000 5000>
    componentLinearAccScale=<0.5 0.5 0.5>
    componentLinearVelScale=<0.5 0.5 0.5>
    gravityOverride {z=980}
    physicsBodyDefinitions [{
      boundBone {boneName=dyn_burger_layer_09}
      boxExtents=<3 30 30>
      constraintSetup {angularLimitsMax=<0 0 6>, angularLimitsMin=<0 0 -6>, angularTarget=<1 0 0>}
      localJointOffset=<2 0 0>
     }]
   }
   n13 = Anim Dynamics - Bone: dyn_burger_layer_01(
    ComponentPose=Local To Component(LocalPose=Sequence Player 'FluteLamp_Stage_02_Idle'() {bLoopAnimation=false, sequence=&FluteLamp_Stage_02_Idle})
    AlphaCurveName=DynamicsAlpha
    GravityOverride="0,0,980"
    ExternalForce="0,0,0"
   ) {
    alphaInputType=Curve
    angularDampingOverride=1
    angularSpringConstant=1000
    bAngularSpring=true
    bOverrideAngularDamping=true
    bUseGravityOverride=true
    gravityScale=0.35
    alphaBoolBlend {blendOption=Linear}
    boundBone {boneName=dyn_burger_layer_01}
    componentAppliedLinearAccClamp=<5000 5000 5000>
    componentLinearAccScale=<0.5 0.5 0.5>
    componentLinearVelScale=<0.5 0.5 0.5>
    gravityOverride {z=980}
    physicsBodyDefinitions [{
      boundBone {boneName=dyn_burger_layer_01}
      boxExtents=<3 30 30>
      constraintSetup {angularLimitsMax=<0 0 1>, angularLimitsMin=<0 0 -1>, angularTarget=<1 0 0>}
      localJointOffset=<2 0 0>
     }]
   }
   n14 = Anim Dynamics - Bone: dyn_burger_layer_06(ComponentPose=@n13, AlphaCurveName=DynamicsAlpha, GravityOverride="0,0,980", ExternalForce="0,0,0") {
    alphaInputType=Curve
    angularDampingOverride=1
    angularSpringConstant=7000
    bAngularSpring=true
    bOverrideAngularDamping=true
    bUseGravityOverride=true
    gravityScale=0.35
    alphaBoolBlend {blendOption=Linear}
    boundBone {boneName=dyn_burger_layer_06}
    componentAppliedLinearAccClamp=<5000 5000 5000>
    componentLinearAccScale=<0.5 0.5 0.5>
    componentLinearVelScale=<0.5 0.5 0.5>
    gravityOverride {z=980}
    physicsBodyDefinitions [{
      boundBone {boneName=dyn_burger_layer_06}
      boxExtents=<3 30 30>
      constraintSetup {angularLimitsMax=<0 0 3>, angularLimitsMin=<0 0 -3>, angularTarget=<1 0 0>}
      localJointOffset=<2 0 0>
     }]
   }
   n15 = Anim Dynamics - Bone: dyn_burger_layer_07(ComponentPose=@n14, AlphaCurveName=DynamicsAlpha, GravityOverride="0,0,980", ExternalForce="0,0,0") {
    alphaInputType=Curve
    angularDampingOverride=1
    angularSpringConstant=7000
    bAngularSpring=true
    bOverrideAngularDamping=true
    bUseGravityOverride=true
    gravityScale=0.35
    alphaBoolBlend {blendOption=Linear}
    boundBone {boneName=dyn_burger_layer_07}
    componentAppliedLinearAccClamp=<5000 5000 5000>
    componentLinearAccScale=<0.5 0.5 0.5>
    componentLinearVelScale=<0.5 0.5 0.5>
    gravityOverride {z=980}
    physicsBodyDefinitions [{
      boundBone {boneName=dyn_burger_layer_07}
      boxExtents=<3 30 30>
      constraintSetup {angularLimitsMax=<0 0 3>, angularLimitsMin=<0 0 -3>, angularTarget=<1 0 0>}
      localJointOffset=<2 0 0>
     }]
   }
   n16 = Anim Dynamics - Bone: dyn_burger_layer_08(ComponentPose=@n15, AlphaCurveName=DynamicsAlpha, GravityOverride="0,0,980", ExternalForce="0,0,0") {
    alphaInputType=Curve
    angularDampingOverride=1
    angularSpringConstant=4000
    bAngularSpring=true
    bOverrideAngularDamping=true
    bUseGravityOverride=true
    gravityScale=0.35
    alphaBoolBlend {blendOption=Linear}
    boundBone {boneName=dyn_burger_layer_08}
    componentAppliedLinearAccClamp=<5000 5000 5000>
    componentLinearAccScale=<0.5 0.5 0.5>
    componentLinearVelScale=<0.5 0.5 0.5>
    gravityOverride {z=980}
    physicsBodyDefinitions [{
      boundBone {boneName=dyn_burger_layer_08}
      boxExtents=<3 30 30>
      constraintSetup {angularLimitsMax=<0 0 5>, angularLimitsMin=<0 0 -5>, angularTarget=<1 0 0>}
      localJointOffset=<2 0 0>
     }]
   }
   n17 = Anim Dynamics - Bone: dyn_burger_layer_09(ComponentPose=@n16, AlphaCurveName=DynamicsAlpha, GravityOverride="0,0,980", ExternalForce="0,0,0") {
    alphaInputType=Curve
    angularDampingOverride=1
    angularSpringConstant=7000
    bAngularSpring=true
    bOverrideAngularDamping=true
    bUseGravityOverride=true
    gravityScale=0.35
    alphaBoolBlend {blendOption=Linear}
    boundBone {boneName=dyn_burger_layer_09}
    componentAppliedLinearAccClamp=<5000 5000 5000>
    componentLinearAccScale=<0.5 0.5 0.5>
    componentLinearVelScale=<0.5 0.5 0.5>
    gravityOverride {z=980}
    physicsBodyDefinitions [{
      boundBone {boneName=dyn_burger_layer_09}
      boxExtents=<3 30 30>
      constraintSetup {angularLimitsMax=<0 0 7>, angularLimitsMin=<0 0 -7>, angularTarget=<1 0 0>}
      localJointOffset=<2 0 0>
     }]
   }
   n18 = Blend Poses by int!(
    BlendPose_0=Component To Local(ComponentPose=@n1)
    BlendPose_1=Component To Local(ComponentPose=@n12)
    BlendPose_2=Component To Local(ComponentPose=@n17)
    BlendPose_3=Component To Local(ComponentPose=@n9)
    BlendPose_4=Component To Local(ComponentPose=@n2)
    BlendTime_0=0
    BlendTime_1=0
    BlendTime_2=0
    BlendTime_3=0
    BlendTime_4=0
    ActiveChildIndex=Get KillCounter() -> (KillCounter=0)
   ) -> (Pose) {
    blendType=Linear
    blendPose [
     {linkId=-1, sourceLinkId=-1}
     {linkId=-1, sourceLinkId=-1}
     {linkId=-1, sourceLinkId=-1}
     {linkId=-1, sourceLinkId=-1}
     {linkId=-1, sourceLinkId=-1}
    ]
    blendTime [0.1,0.1,0.1,0.1,0.1]
   }
   Output Animation Pose(Result=@n18)
  ]
 }
 OnKill {
  path=AnimGraph.AnimGraphNode_StateMachine_1.OnKillStateMachine.AnimStateNode_1.OnKill
  nodes [
   Anim Dynamics - Bone: dyn_burger_layer_01(AlphaCurveName=DynamicsAlpha, GravityOverride="0,0,980", ExternalForce="0,0,0") {
    alphaInputType=Curve
    angularDampingOverride=0.9999
    angularSpringConstant=7000
    bAngularSpring=true
    bOverrideAngularDamping=true
    bUseGravityOverride=true
    gravityScale=0.35
    alphaBoolBlend {blendOption=Linear}
    boundBone {boneName=dyn_burger_layer_01}
    componentAppliedLinearAccClamp=<5000 5000 5000>
    componentLinearAccScale=<0.5 0.5 0.5>
    componentLinearVelScale=<0.5 0.5 0.5>
    gravityOverride {z=980}
    physicsBodyDefinitions [{
      boundBone {boneName=dyn_burger_layer_01}
      boxExtents=<3 30 30>
      constraintSetup {angularLimitsMax=<0 0 7>, angularLimitsMin=<0 0 -7>, angularTarget=<1 0 0>}
      localJointOffset=<3 0 0>
     }]
   }
   n1 = Sequence Player 'FluteLamp_Stage04Loop_Reactive'() {bLoopAnimation=false, sequence=&FluteLamp_Stage04Loop_Reactive}
   n2 = RigidBody(ComponentPose=Local To Component(LocalPose=@n1), AlphaCurveName=DynamicsAlpha, OverrideWorldGravity="0,0,980") {
    alphaInputType=Curve
    bOverrideWorldGravity=true
    overridePhysicsAsset=&M_MED_FluteLamp_Pack_Physics
    simulationSpace=BaseBoneSpace
    alphaBoolBlend {blendOption=Linear}
    baseBoneRef {boneName=pack}
    componentAppliedLinearAccClamp {y=0}
    overrideWorldGravity {z=980}
    simSpaceSettings {
     dampingAlpha=0
     maxAngularAcceleration=3
     maxAngularVelocity=1
     maxLinearAcceleration=500
     maxLinearVelocity=200
     velocityScaleZ=0
     worldAlpha=0.5
     externalLinearDragV {x=3, z=1}
    }
   }
   Component To Local(ComponentPose=@n2)
   n3 = Sequence Player 'FluteLamp_Stage01_Reactive'() {bLoopAnimation=false, sequence=&FluteLamp_Stage01_Reactive}
   n4 = Anim Dynamics - Bone: dyn_burger_layer_09(ComponentPose=Local To Component(LocalPose=@n3), AlphaCurveName=DynamicsAlpha, GravityOverride="0,0,980", ExternalForce="0,0,0") {
    alphaInputType=Curve
    angularDampingOverride=0.9999
    angularSpringConstant=7000
    bAngularSpring=true
    bOverrideAngularDamping=true
    bUseGravityOverride=true
    gravityScale=0.35
    alphaBoolBlend {blendOption=Linear}
    boundBone {boneName=dyn_burger_layer_09}
    componentAppliedLinearAccClamp=<5000 5000 5000>
    componentLinearAccScale=<0.5 0.5 0.5>
    componentLinearVelScale=<0.5 0.5 0.5>
    gravityOverride {z=980}
    physicsBodyDefinitions [{
      boundBone {boneName=dyn_burger_layer_09}
      boxExtents=<3 30 30>
      constraintSetup {angularLimitsMax=<0 0 6>, angularLimitsMin=<0 0 -6>, angularTarget=<1 0 0>}
      localJointOffset=<2 0 0>
     }]
   }
   n5 = Sequence Player 'FluteLamp_Stage04Loop_Reactive'() {bLoopAnimation=false, sequence=&FluteLamp_Stage04Loop_Reactive}
   n6 = Anim Dynamics - Bone: dyn_burger_layer_01(ComponentPose=Local To Component(LocalPose=@n5), AlphaCurveName=DynamicsAlpha, GravityOverride="0,0,980", ExternalForce="0,0,0") {
    alphaInputType=Curve
    angularDampingOverride=1
    angularSpringConstant=1000
    bAngularSpring=true
    bOverrideAngularDamping=true
    bUseGravityOverride=true
    gravityScale=0.35
    alphaBoolBlend {blendOption=Linear}
    boundBone {boneName=dyn_burger_layer_01}
    componentAppliedLinearAccClamp=<5000 5000 5000>
    componentLinearAccScale=<0.5 0.5 0.5>
    componentLinearVelScale=<0.5 0.5 0.5>
    gravityOverride {z=980}
    physicsBodyDefinitions [{
      boundBone {boneName=dyn_burger_layer_01}
      boxExtents=<3 30 30>
      constraintSetup {angularLimitsMax=<0 0 1>, angularLimitsMin=<0 0 -1>, angularTarget=<1 0 0>}
      localJointOffset=<2 0 0>
     }]
   }
   n7 = Anim Dynamics - Bone: dyn_burger_layer_04(ComponentPose=@n6, AlphaCurveName=DynamicsAlpha, GravityOverride="0,0,980", ExternalForce="0,0,0") {
    alphaInputType=Curve
    angularDampingOverride=1
    angularSpringConstant=3000
    bAngularSpring=true
    bOverrideAngularDamping=true
    bUseGravityOverride=true
    gravityScale=0.35
    alphaBoolBlend {blendOption=Linear}
    boundBone {boneName=dyn_burger_layer_04}
    componentAppliedLinearAccClamp=<5000 5000 5000>
    componentLinearAccScale=<0.5 0.5 0.5>
    componentLinearVelScale=<0.5 0.5 0.5>
    gravityOverride {z=980}
    physicsBodyDefinitions [{
      boundBone {boneName=dyn_burger_layer_04}
      boxExtents=<3 30 30>
      constraintSetup {angularLimitsMax=<0 0 2>, angularLimitsMin=<0 0 -2>, angularTarget=<1 0 0>}
      localJointOffset=<2 0 0>
     }]
   }
   n8 = Anim Dynamics - Bone: dyn_burger_layer_05(ComponentPose=@n7, AlphaCurveName=DynamicsAlpha, GravityOverride="0,0,980", ExternalForce="0,0,0") {
    alphaInputType=Curve
    angularDampingOverride=1
    angularSpringConstant=5000
    bAngularSpring=true
    bOverrideAngularDamping=true
    bUseGravityOverride=true
    gravityScale=0.35
    alphaBoolBlend {blendOption=Linear}
    boundBone {boneName=dyn_burger_layer_05}
    componentAppliedLinearAccClamp=<5000 5000 5000>
    componentLinearAccScale=<0.5 0.5 0.5>
    componentLinearVelScale=<0.5 0.5 0.5>
    gravityOverride {z=980}
    physicsBodyDefinitions [{
      boundBone {boneName=dyn_burger_layer_05}
      boxExtents=<3 30 30>
      constraintSetup {angularLimitsMax=<0 0 5>, angularLimitsMin=<0 0 -5>, angularTarget=<1 0 0>}
      localJointOffset=<2 0 0>
     }]
   }
   n9 = Anim Dynamics - Bone: dyn_burger_layer_06(ComponentPose=@n8, AlphaCurveName=DynamicsAlpha, GravityOverride="0,0,980", ExternalForce="0,0,0") {
    alphaInputType=Curve
    angularDampingOverride=1
    angularSpringConstant=3000
    bAngularSpring=true
    bOverrideAngularDamping=true
    bUseGravityOverride=true
    gravityScale=0.35
    alphaBoolBlend {blendOption=Linear}
    boundBone {boneName=dyn_burger_layer_06}
    componentAppliedLinearAccClamp=<5000 5000 5000>
    componentLinearAccScale=<0.5 0.5 0.5>
    componentLinearVelScale=<0.5 0.5 0.5>
    gravityOverride {z=980}
    physicsBodyDefinitions [{
      boundBone {boneName=dyn_burger_layer_06}
      boxExtents=<5 30 30>
      constraintSetup {angularLimitsMax=<0 0 5>, angularLimitsMin=<0 0 -5>, angularTarget=<1 0 0>}
      localJointOffset=<2 0 0>
     }]
   }
   n10 = Anim Dynamics - Bone: dyn_burger_layer_07(ComponentPose=@n9, AlphaCurveName=DynamicsAlpha, GravityOverride="0,0,980", ExternalForce="0,0,0") {
    alphaInputType=Curve
    angularDampingOverride=1
    angularSpringConstant=3000
    bAngularSpring=true
    bOverrideAngularDamping=true
    bUseGravityOverride=true
    gravityScale=0.35
    alphaBoolBlend {blendOption=Linear}
    boundBone {boneName=dyn_burger_layer_07}
    componentAppliedLinearAccClamp=<5000 5000 5000>
    componentLinearAccScale=<0.5 0.5 0.5>
    componentLinearVelScale=<0.5 0.5 0.5>
    gravityOverride {z=980}
    physicsBodyDefinitions [{
      boundBone {boneName=dyn_burger_layer_07}
      boxExtents=<10 30 30>
      constraintSetup {angularLimitsMax=<0 0 5>, angularLimitsMin=<0 0 -5>, angularTarget=<1 0 0>}
      localJointOffset=<2 0 0>
     }]
   }
   n11 = Anim Dynamics - Bone: dyn_burger_layer_08(ComponentPose=@n10, AlphaCurveName=DynamicsAlpha, GravityOverride="0,0,980", ExternalForce="0,0,0") {
    alphaInputType=Curve
    angularDampingOverride=1
    angularSpringConstant=7000
    bAngularSpring=true
    bOverrideAngularDamping=true
    bUseGravityOverride=true
    gravityScale=0.35
    alphaBoolBlend {blendOption=Linear}
    boundBone {boneName=dyn_burger_layer_08}
    componentAppliedLinearAccClamp=<5000 5000 5000>
    componentLinearAccScale=<0.5 0.5 0.5>
    componentLinearVelScale=<0.5 0.5 0.5>
    gravityOverride {z=980}
    physicsBodyDefinitions [{
      boundBone {boneName=dyn_burger_layer_08}
      boxExtents=<3 30 30>
      constraintSetup {angularLimitsMax=<0 0 8>, angularLimitsMin=<0 0 -8>, angularTarget=<1 0 0>}
      localJointOffset=<2 0 0>
     }]
   }
   n12 = Anim Dynamics - Bone: dyn_burger_layer_09(ComponentPose=@n11, AlphaCurveName=DynamicsAlpha, GravityOverride="0,0,980", ExternalForce="0,0,0") {
    alphaInputType=Curve
    angularDampingOverride=1
    angularSpringConstant=7000
    bAngularSpring=true
    bOverrideAngularDamping=true
    bUseGravityOverride=true
    gravityScale=0.35
    alphaBoolBlend {blendOption=Linear}
    boundBone {boneName=dyn_burger_layer_09}
    componentAppliedLinearAccClamp=<5000 5000 5000>
    componentLinearAccScale=<0.5 0.5 0.5>
    componentLinearVelScale=<0.5 0.5 0.5>
    gravityOverride {z=980}
    physicsBodyDefinitions [{
      boundBone {boneName=dyn_burger_layer_09}
      boxExtents=<3 30 30>
      constraintSetup {angularLimitsMax=<0 0 8>, angularLimitsMin=<0 0 -8>, angularTarget=<1 0 0>}
      localJointOffset=<2 0 0>
     }]
   }
   n13 = Sequence Player 'FluteLamp_Stage04_Reactive'() {bLoopAnimation=false, sequence=&FluteLamp_Stage04_Reactive}
   n14 = Anim Dynamics - Bone: dyn_burger_layer_01(ComponentPose=Local To Component(LocalPose=@n13), AlphaCurveName=DynamicsAlpha, GravityOverride="0,0,980", ExternalForce="0,0,0") {
    alphaInputType=Curve
    angularDampingOverride=1
    angularSpringConstant=1000
    bAngularSpring=true
    bOverrideAngularDamping=true
    bUseGravityOverride=true
    gravityScale=0.35
    alphaBoolBlend {blendOption=Linear}
    boundBone {boneName=dyn_burger_layer_01}
    componentAppliedLinearAccClamp=<5000 5000 5000>
    componentLinearAccScale=<0.5 0.5 0.5>
    componentLinearVelScale=<0.5 0.5 0.5>
    gravityOverride {z=980}
    physicsBodyDefinitions [{
      boundBone {boneName=dyn_burger_layer_01}
      boxExtents=<3 30 30>
      constraintSetup {angularLimitsMax=<0 0 1>, angularLimitsMin=<0 0 -1>, angularTarget=<1 0 0>}
      localJointOffset=<2 0 0>
     }]
   }
   n15 = Anim Dynamics - Bone: dyn_burger_layer_04(ComponentPose=@n14, AlphaCurveName=DynamicsAlpha, GravityOverride="0,0,980", ExternalForce="0,0,0") {
    alphaInputType=Curve
    angularDampingOverride=1
    angularSpringConstant=3000
    bAngularSpring=true
    bOverrideAngularDamping=true
    bUseGravityOverride=true
    gravityScale=0.35
    alphaBoolBlend {blendOption=Linear}
    boundBone {boneName=dyn_burger_layer_04}
    componentAppliedLinearAccClamp=<5000 5000 5000>
    componentLinearAccScale=<0.5 0.5 0.5>
    componentLinearVelScale=<0.5 0.5 0.5>
    gravityOverride {z=980}
    physicsBodyDefinitions [{
      boundBone {boneName=dyn_burger_layer_04}
      boxExtents=<3 30 30>
      constraintSetup {angularLimitsMax=<0 0 2>, angularLimitsMin=<0 0 -2>, angularTarget=<1 0 0>}
      localJointOffset=<2 0 0>
     }]
   }
   n16 = Anim Dynamics - Bone: dyn_burger_layer_05(ComponentPose=@n15, AlphaCurveName=DynamicsAlpha, GravityOverride="0,0,980", ExternalForce="0,0,0") {
    alphaInputType=Curve
    angularDampingOverride=1
    angularSpringConstant=5000
    bAngularSpring=true
    bOverrideAngularDamping=true
    bUseGravityOverride=true
    gravityScale=0.35
    alphaBoolBlend {blendOption=Linear}
    boundBone {boneName=dyn_burger_layer_05}
    componentAppliedLinearAccClamp=<5000 5000 5000>
    componentLinearAccScale=<0.5 0.5 0.5>
    componentLinearVelScale=<0.5 0.5 0.5>
    gravityOverride {z=980}
    physicsBodyDefinitions [{
      boundBone {boneName=dyn_burger_layer_05}
      boxExtents=<3 30 30>
      constraintSetup {angularLimitsMax=<0 0 5>, angularLimitsMin=<0 0 -5>, angularTarget=<1 0 0>}
      localJointOffset=<2 0 0>
     }]
   }
   n17 = Anim Dynamics - Bone: dyn_burger_layer_06(ComponentPose=@n16, AlphaCurveName=DynamicsAlpha, GravityOverride="0,0,980", ExternalForce="0,0,0") {
    alphaInputType=Curve
    angularDampingOverride=1
    angularSpringConstant=3000
    bAngularSpring=true
    bOverrideAngularDamping=true
    bUseGravityOverride=true
    gravityScale=0.35
    alphaBoolBlend {blendOption=Linear}
    boundBone {boneName=dyn_burger_layer_06}
    componentAppliedLinearAccClamp=<5000 5000 5000>
    componentLinearAccScale=<0.5 0.5 0.5>
    componentLinearVelScale=<0.5 0.5 0.5>
    gravityOverride {z=980}
    physicsBodyDefinitions [{
      boundBone {boneName=dyn_burger_layer_06}
      boxExtents=<5 30 30>
      constraintSetup {angularLimitsMax=<0 0 5>, angularLimitsMin=<0 0 -5>, angularTarget=<1 0 0>}
      localJointOffset=<2 0 0>
     }]
   }
   n18 = Anim Dynamics - Bone: dyn_burger_layer_07(ComponentPose=@n17, AlphaCurveName=DynamicsAlpha, GravityOverride="0,0,980", ExternalForce="0,0,0") {
    alphaInputType=Curve
    angularDampingOverride=1
    angularSpringConstant=3000
    bAngularSpring=true
    bOverrideAngularDamping=true
    bUseGravityOverride=true
    gravityScale=0.35
    alphaBoolBlend {blendOption=Linear}
    boundBone {boneName=dyn_burger_layer_07}
    componentAppliedLinearAccClamp=<5000 5000 5000>
    componentLinearAccScale=<0.5 0.5 0.5>
    componentLinearVelScale=<0.5 0.5 0.5>
    gravityOverride {z=980}
    physicsBodyDefinitions [{
      boundBone {boneName=dyn_burger_layer_07}
      boxExtents=<10 30 30>
      constraintSetup {angularLimitsMax=<0 0 5>, angularLimitsMin=<0 0 -5>, angularTarget=<1 0 0>}
      localJointOffset=<2 0 0>
     }]
   }
   n19 = Anim Dynamics - Bone: dyn_burger_layer_08(ComponentPose=@n18, AlphaCurveName=DynamicsAlpha, GravityOverride="0,0,980", ExternalForce="0,0,0") {
    alphaInputType=Curve
    angularDampingOverride=1
    angularSpringConstant=7000
    bAngularSpring=true
    bOverrideAngularDamping=true
    bUseGravityOverride=true
    gravityScale=0.35
    alphaBoolBlend {blendOption=Linear}
    boundBone {boneName=dyn_burger_layer_08}
    componentAppliedLinearAccClamp=<5000 5000 5000>
    componentLinearAccScale=<0.5 0.5 0.5>
    componentLinearVelScale=<0.5 0.5 0.5>
    gravityOverride {z=980}
    physicsBodyDefinitions [{
      boundBone {boneName=dyn_burger_layer_08}
      boxExtents=<3 30 30>
      constraintSetup {angularLimitsMax=<0 0 8>, angularLimitsMin=<0 0 -8>, angularTarget=<1 0 0>}
      localJointOffset=<2 0 0>
     }]
   }
   n20 = Anim Dynamics - Bone: dyn_burger_layer_09(ComponentPose=@n19, AlphaCurveName=DynamicsAlpha, GravityOverride="0,0,980", ExternalForce="0,0,0") {
    alphaInputType=Curve
    angularDampingOverride=1
    angularSpringConstant=7000
    bAngularSpring=true
    bOverrideAngularDamping=true
    bUseGravityOverride=true
    gravityScale=0.35
    alphaBoolBlend {blendOption=Linear}
    boundBone {boneName=dyn_burger_layer_09}
    componentAppliedLinearAccClamp=<5000 5000 5000>
    componentLinearAccScale=<0.5 0.5 0.5>
    componentLinearVelScale=<0.5 0.5 0.5>
    gravityOverride {z=980}
    physicsBodyDefinitions [{
      boundBone {boneName=dyn_burger_layer_09}
      boxExtents=<3 30 30>
      constraintSetup {angularLimitsMax=<0 0 8>, angularLimitsMin=<0 0 -8>, angularTarget=<1 0 0>}
      localJointOffset=<2 0 0>
     }]
   }
   n21 = Sequence Player 'FluteLamp_Stage02_Reactive'() {bLoopAnimation=false, sequence=&FluteLamp_Stage02_Reactive}
   n22 = Anim Dynamics - Bone: dyn_burger_layer_01(ComponentPose=Local To Component(LocalPose=@n21), AlphaCurveName=DynamicsAlpha, GravityOverride="0,0,980", ExternalForce="0,0,0") {
    alphaInputType=Curve
    angularDampingOverride=1
    angularSpringConstant=7000
    bAngularSpring=true
    bOverrideAngularDamping=true
    bUseGravityOverride=true
    gravityScale=0.35
    alphaBoolBlend {blendOption=Linear}
    boundBone {boneName=dyn_burger_layer_01}
    componentAppliedLinearAccClamp=<5000 5000 5000>
    componentLinearAccScale=<0.5 0.5 0.5>
    componentLinearVelScale=<0.5 0.5 0.5>
    gravityOverride {z=980}
    physicsBodyDefinitions [{
      boundBone {boneName=dyn_burger_layer_01}
      boxExtents=<3 30 30>
      constraintSetup {angularLimitsMax=<0 0 1>, angularLimitsMin=<0 0 -1>, angularTarget=<1 0 0>}
      localJointOffset=<2 0 0>
     }]
   }
   n23 = Anim Dynamics - Bone: dyn_burger_layer_08(ComponentPose=@n22, AlphaCurveName=DynamicsAlpha, GravityOverride="0,0,980", ExternalForce="0,0,0") {
    alphaInputType=Curve
    angularDampingOverride=0.9999
    angularSpringConstant=2000
    bAngularSpring=true
    bOverrideAngularDamping=true
    bUseGravityOverride=true
    gravityScale=0.35
    alphaBoolBlend {blendOption=Linear}
    boundBone {boneName=dyn_burger_layer_08}
    componentAppliedLinearAccClamp=<5000 5000 5000>
    componentLinearAccScale=<0.5 0.5 0.5>
    componentLinearVelScale=<0.5 0.5 0.5>
    gravityOverride {z=980}
    physicsBodyDefinitions [{
      boundBone {boneName=dyn_burger_layer_08}
      boxExtents=<11 30 30>
      constraintSetup {angularLimitsMax=<0 0 6>, angularLimitsMin=<0 0 -6>, angularTarget=<1 0 0>}
      localJointOffset=<2 0 0>
     }]
   }
   n24 = Anim Dynamics - Bone: dyn_burger_layer_09(ComponentPose=@n23, AlphaCurveName=DynamicsAlpha, GravityOverride="0,0,980", ExternalForce="0,0,0") {
    alphaInputType=Curve
    angularDampingOverride=1
    angularSpringConstant=5000
    bAngularSpring=true
    bOverrideAngularDamping=true
    bUseGravityOverride=true
    gravityScale=0.35
    alphaBoolBlend {blendOption=Linear}
    boundBone {boneName=dyn_burger_layer_09}
    componentAppliedLinearAccClamp=<5000 5000 5000>
    componentLinearAccScale=<0.5 0.5 0.5>
    componentLinearVelScale=<0.5 0.5 0.5>
    gravityOverride {z=980}
    physicsBodyDefinitions [{
      boundBone {boneName=dyn_burger_layer_09}
      boxExtents=<3 30 30>
      constraintSetup {angularLimitsMax=<0 0 6>, angularLimitsMin=<0 0 -6>, angularTarget=<1 0 0>}
      localJointOffset=<2 0 0>
     }]
   }
   n25 = Sequence Player 'FluteLamp_Stage03_Reactive'() {bLoopAnimation=false, sequence=&FluteLamp_Stage03_Reactive}
   n26 = Anim Dynamics - Bone: dyn_burger_layer_01(ComponentPose=Local To Component(LocalPose=@n25), AlphaCurveName=DynamicsAlpha, GravityOverride="0,0,980", ExternalForce="0,0,0") {
    alphaInputType=Curve
    angularDampingOverride=1
    angularSpringConstant=1000
    bAngularSpring=true
    bOverrideAngularDamping=true
    bUseGravityOverride=true
    gravityScale=0.35
    alphaBoolBlend {blendOption=Linear}
    boundBone {boneName=dyn_burger_layer_01}
    componentAppliedLinearAccClamp=<5000 5000 5000>
    componentLinearAccScale=<0.5 0.5 0.5>
    componentLinearVelScale=<0.5 0.5 0.5>
    gravityOverride {z=980}
    physicsBodyDefinitions [{
      boundBone {boneName=dyn_burger_layer_01}
      boxExtents=<3 30 30>
      constraintSetup {angularLimitsMax=<0 0 1>, angularLimitsMin=<0 0 -1>, angularTarget=<1 0 0>}
      localJointOffset=<2 0 0>
     }]
   }
   n27 = Anim Dynamics - Bone: dyn_burger_layer_06(ComponentPose=@n26, AlphaCurveName=DynamicsAlpha, GravityOverride="0,0,980", ExternalForce="0,0,0") {
    alphaInputType=Curve
    angularDampingOverride=1
    angularSpringConstant=7000
    bAngularSpring=true
    bOverrideAngularDamping=true
    bUseGravityOverride=true
    gravityScale=0.35
    alphaBoolBlend {blendOption=Linear}
    boundBone {boneName=dyn_burger_layer_06}
    componentAppliedLinearAccClamp=<5000 5000 5000>
    componentLinearAccScale=<0.5 0.5 0.5>
    componentLinearVelScale=<0.5 0.5 0.5>
    gravityOverride {z=980}
    physicsBodyDefinitions [{
      boundBone {boneName=dyn_burger_layer_06}
      boxExtents=<3 30 30>
      constraintSetup {angularLimitsMax=<0 0 3>, angularLimitsMin=<0 0 -3>, angularTarget=<1 0 0>}
      localJointOffset=<2 0 0>
     }]
   }
   n28 = Anim Dynamics - Bone: dyn_burger_layer_07(ComponentPose=@n27, AlphaCurveName=DynamicsAlpha, GravityOverride="0,0,980", ExternalForce="0,0,0") {
    alphaInputType=Curve
    angularDampingOverride=1
    angularSpringConstant=7000
    bAngularSpring=true
    bOverrideAngularDamping=true
    bUseGravityOverride=true
    gravityScale=0.35
    alphaBoolBlend {blendOption=Linear}
    boundBone {boneName=dyn_burger_layer_07}
    componentAppliedLinearAccClamp=<5000 5000 5000>
    componentLinearAccScale=<0.5 0.5 0.5>
    componentLinearVelScale=<0.5 0.5 0.5>
    gravityOverride {z=980}
    physicsBodyDefinitions [{
      boundBone {boneName=dyn_burger_layer_07}
      boxExtents=<3 30 30>
      constraintSetup {angularLimitsMax=<0 0 3>, angularLimitsMin=<0 0 -3>, angularTarget=<1 0 0>}
      localJointOffset=<2 0 0>
     }]
   }
   n29 = Anim Dynamics - Bone: dyn_burger_layer_08(ComponentPose=@n28, AlphaCurveName=DynamicsAlpha, GravityOverride="0,0,980", ExternalForce="0,0,0") {
    alphaInputType=Curve
    angularDampingOverride=1
    angularSpringConstant=4000
    bAngularSpring=true
    bOverrideAngularDamping=true
    bUseGravityOverride=true
    gravityScale=0.35
    alphaBoolBlend {blendOption=Linear}
    boundBone {boneName=dyn_burger_layer_08}
    componentAppliedLinearAccClamp=<5000 5000 5000>
    componentLinearAccScale=<0.5 0.5 0.5>
    componentLinearVelScale=<0.5 0.5 0.5>
    gravityOverride {z=980}
    physicsBodyDefinitions [{
      boundBone {boneName=dyn_burger_layer_08}
      boxExtents=<3 30 30>
      constraintSetup {angularLimitsMax=<0 0 5>, angularLimitsMin=<0 0 -5>, angularTarget=<1 0 0>}
      localJointOffset=<2 0 0>
     }]
   }
   n30 = Anim Dynamics - Bone: dyn_burger_layer_09(ComponentPose=@n29, AlphaCurveName=DynamicsAlpha, GravityOverride="0,0,980", ExternalForce="0,0,0") {
    alphaInputType=Curve
    angularDampingOverride=1
    angularSpringConstant=7000
    bAngularSpring=true
    bOverrideAngularDamping=true
    bUseGravityOverride=true
    gravityScale=0.35
    alphaBoolBlend {blendOption=Linear}
    boundBone {boneName=dyn_burger_layer_09}
    componentAppliedLinearAccClamp=<5000 5000 5000>
    componentLinearAccScale=<0.5 0.5 0.5>
    componentLinearVelScale=<0.5 0.5 0.5>
    gravityOverride {z=980}
    physicsBodyDefinitions [{
      boundBone {boneName=dyn_burger_layer_09}
      boxExtents=<3 30 30>
      constraintSetup {angularLimitsMax=<0 0 7>, angularLimitsMin=<0 0 -7>, angularTarget=<1 0 0>}
      localJointOffset=<2 0 0>
     }]
   }
   n31 = Blend Poses by int!(
    BlendPose_0
    BlendPose_1=Component To Local(ComponentPose=@n4)
    BlendPose_2=Component To Local(ComponentPose=@n24)
    BlendPose_3=Component To Local(ComponentPose=@n30)
    BlendPose_4=Component To Local(ComponentPose=@n20)
    BlendPose_5:Pose Link Structure=Component To Local(ComponentPose=@n12)
    BlendTime_0=0
    BlendTime_1=0
    BlendTime_2=0
    BlendTime_3=0
    BlendTime_4=0
    BlendTime_5:float=0
    ActiveChildIndex=Get KillCounter() -> (KillCounter=0)
   ) -> (Pose) {
    blendType=Linear
    blendPose [
     {linkId=-1, sourceLinkId=-1}
     {linkId=-1, sourceLinkId=-1}
     {linkId=-1, sourceLinkId=-1}
     {linkId=-1, sourceLinkId=-1}
     {linkId=-1, sourceLinkId=-1}
     {linkId=-1, sourceLinkId=-1}
    ]
    blendTime [0.1,0.1,0.1,0.1,0.1,0.1]
   }
   Output Animation Pose(Result=@n31)
  ]
 }
 Transition {
  path=AnimGraph.AnimGraphNode_StateMachine_1.OnKillStateMachine.AnimStateTransitionNode_0.Transition
  nodes [
   n1 = Get ActiveKill()
   Result(bCanEnterTransition=@n1)
   AND Boolean(A=@n1, B="integer < integer"(A=Get KillCounter() -> (KillCounter=0), B=Get MaxKills() -> (MaxKills=0)))
  ]
 }
 Transition {path=AnimGraph.AnimGraphNode_StateMachine_1.OnKillStateMachine.AnimStateTransitionNode_1.Transition, nodes [Result()]}
 EventGraph {nodes [
   n1 = Timer Exec() {is=title K2Node_CustomEvent}
   > Branch(Condition=Get CosmeticVariantReactivity())
   > UpdatePattyVisibilityInFrontend()
   Event BlueprintUpdateAnimation()
   > n2 = Sequence!(execute) -> (then_0, then_1, then_2:Exec, then_3:Exec)
   > Set GravityOverride(GravityOverride="0,0,980")
   n3 = Branch(execute=@n2.then_1, Condition=Get bIsSkyDiving())
   > Set GravityOverride(GravityOverride="0,0,200")
   > Branch(Condition=Get bIsSkydiveDiveMode())
   > Set GravityOverride(GravityOverride="0,0,-980")
   n4 = Branch(execute=@n3.else, Condition=Get bIsTacticalSprinting())
   > Set GravityOverride(GravityOverride="0,200,500")
   Branch(execute=@n4.else, Condition=AND Boolean(A=Get bIsCrouching(), B=Get bIsMoving2D()))
   > Set GravityOverride(GravityOverride="0,0,500")
   n5 = Sequence!(execute=@n2.then_2) -> (then_0, then_1, then_2)
   n6 = Get bodyMesh() -> (bodyMesh:Skeletal Mesh Component Object Reference)
   > Is Valid(InputObject=@n6)
   > Set PelvisLocation(PelvisLocation:Vector=GetSocketLocation!(self:Scene Component Object Reference=@n6, InSocketName:Name=pelvis) -> (ReturnValue:Vector)) -> (Output_Get:Vector)
   > Branch(Condition=Get bIsInFrontEnd())
   > Branch(Condition="integer >= integer"(A=Get KillCounter() -> (KillCounter=0), B=4))
   > Do Once(Reset=@n11.then)
   > n7 = Delay(Duration=40)
   Do Once(execute=@n26.then, Reset=@n7.then)
   > n8 = Sequence!(execute) -> (then_0, then_1)
   > Set activeKill(activeKill:bool=false) -> (Output_Get:bool)
   > Set KillCounter(KillCounter:int=0) -> (Output_Get:int=0)
   > n9 = Set Timer by Event(Delegate=@n1, Time=Get Timer() -> (Timer:float), bLooping=true)
   > n10 = Set TimerHandle(TimerHandle:Timer Handle Structure=@n9) -> (Output_Get:Timer Handle Structure)
   > n11 = Unpause Timer by Handle!(execute, Handle:Timer Handle Structure=@n10) -> (then)
   n12 = GetPartSkeletalMeshComponent()
   n13 = Get Dynamic Material Instance(execute=@n8.then_1, SkeletalMeshComponent=@n12)
   > n14 = Set DynamicMaterialInstance(DynamicMaterialInstance:Material Instance Dynamic Object Reference=@n13) -> (Output_Get:Material Instance Dynamic Object Reference)
   Set SkeletalMesh(execute=[@n27.then_0, @n14.then], SkeletalMesh:Skeletal Mesh Component Object Reference=@n12) -> (Output_Get:Skeletal Mesh Component Object Reference)
   > n15 = Cast To FortPlayerPawnAthena(Object=TryGetPawnOwner())
   > Set bodyMesh(bodyMesh:Skeletal Mesh Component Object Reference=GetSkeletalMeshForPartType(self=@n15)) -> (Output_Get:Skeletal Mesh Component Object Reference)
   > Branch(Condition=Get bIsInFrontEnd())
   n16 = Get DynamicMaterialInstance() -> (DynamicMaterialInstance:Material Instance Dynamic Object Reference)
   > SetScalarParameterValue(self=@n16, ParameterName=Hide Element 02, Value=1)
   > SetScalarParameterValue(self=@n16, ParameterName=Hide Element 03, Value=1)
   Branch(execute=@n5.then_2, Condition=Get bIsInFrontEnd())
   > Do Once(Reset=[@n23.then, @n24.then, @n21.0, @n21.Default])
   > n17 = Branch(Condition="integer > integer"(A=Get KillCounter() -> (KillCounter=0), B=3))
   > n18 = Delay(Duration=10)
   Branch(execute=[@n18.then, @n25.then, @n35.then], Condition=Get CosmeticVariantReactivity())
   > RetriggerableDelay(Duration=0.6)
   > Set activeKill(activeKill=true)
   > n19 = Increment Int(Value=Get KillCounter() -> (KillCounter=0))
   > n20 = Set KillCounter(KillCounter=@n19) -> (Output_Get=0)
   > n21 = Switch on Int!(execute:Exec, Selection:int=@n20) -> (Default:Exec, 0:Exec, 1:Exec, 2:Exec)
   n22 = Get DynamicMaterialInstance()
   n23 = SetScalarParameterValue(execute=@n21.1, self=@n22, ParameterName=Hide Element 02, Value=0)
   n24 = SetScalarParameterValue(execute=@n21.2, self=@n22, ParameterName=Hide Element 03, Value=0)
   n25 = Delay(execute=@n17.else, Duration=1.5)
   n26 = Branch(execute=@n5.then_1, Condition=Get bIsInFrontEnd())
   Event OnCharacterCustomizationLoaded()
   > n27 = Sequence!(execute) -> (then_0, then_1, then_2:Exec)
   n28 = Cast To FortPlayerPawn(execute=@n27.then_1, Object=TryGetFortPawn())
   > n29 = Cast To B_Athena_PartModifier_Backpack_FluteLamp!(
    execute
    Object=GetModifierForPartType!(self:Fort Player Pawn Object Reference=@n28, PartType:EFortCustomPartType Enum=Backpack) -> (ReturnValue:Custom Character Part Modifier Object Reference)
   ) -> (then, CastFailed, AsB Athena Part Modifier Backpack Flute Lamp:B Athena Part Modifier Backpack Flute Lamp Object Reference) {is=title K2Node_DynamicCast}
   > Set CosmeticVariantReactivity(CosmeticVariantReactivity:bool=Get ReactivityEnabled!(self:B Athena Part Modifier Backpack Flute Lamp Object Reference=@n29) -> (ReactivityEnabled:bool)) -> (Output_Get:bool)
   n30 = Get Dynamic Material Instance(execute=@n27.then_2, SkeletalMeshComponent=GetPartSkeletalMeshComponent())
   > Set DynamicMaterialInstance(DynamicMaterialInstance=@n30)
   > n31 = Switch on Int(Selection=Get KillCounter() -> (KillCounter=0))
   n32 = Get DynamicMaterialInstance()
   n33 = SetScalarParameterValue(execute=@n31.0, self=@n32, ParameterName=Hide Element 02, Value=1)
   SetScalarParameterValue(execute=[@n33.then, @n31.1], self=@n32, ParameterName=Hide Element 03, Value=1)
   n34 = Event BlueprintProcessAnimInputEvent()
   > Sequence!(execute) -> (then_0, then_1)
   > n35 = Branch(Condition=IsNamed(self=@n34.AnimInputEvent, InEventName=aie.EnemyPlayerEliminated))
   Event BlueprintPostEvaluateAnimation() {is=title K2Node_Event}
   > Branch(Condition=Get CosmeticVariantReactivity())
   > Set activeKill(activeKill=false)
   TryGetPawnOwner()
   Set KillCounter(KillCounter=0) -> (Output_Get=0)
  ]}
 UpdatePattyVisibilityInFrontend {nodes [
   UpdatePattyVisibilityInFrontend() {is=title K2Node_FunctionEntry}
   > Set activeKill(activeKill=true)
   n1 = Get KillCounter() -> (KillCounter=0)
   n2 = Get MaxKills() -> (MaxKills=0)
   > Set KillCounter(KillCounter="Clamp (Integer)"(Value=int + int(A=@n1), Min=0, Max=int + int(A=@n2)) -> (ReturnValue=0)) -> (Output_Get=0)
   > n3 = Switch on Int(Selection=Get KillCounter() -> (KillCounter=0))
   n4 = Get DynamicMaterialInstance()
   SetScalarParameterValue(execute=@n3.1, self=@n4, ParameterName=Hide Element 02, Value=0)
   SetScalarParameterValue(execute=@n3.2, self=@n4, ParameterName=Hide Element 03, Value=0)
   > SetScalarParameterValue(self=@n4, ParameterName=Hide Element 04, Value=0)
  ]}
]
variables [
 ActiveKill {type=boolean}
 BodyMesh {type=SkeletalMeshComponent}
 BurgerStage_01 {type=boolean}
 BurgerStage_02 {type=boolean}
 BurgerStage_03 {type=boolean}
 BurgerStage_04 {type=boolean}
 BurgerStage_05 {type=boolean}
 CosmeticVariantReactivity {type=boolean}
 DynamicMaterialInstance {type=MaterialInstanceDynamic}
 GravityOverride {type=Vector}
 KillCounter {type=integer}
 MaxKills {type=integer, default=5}
 PelvisLocation {type=Vector}
 SkeletalMesh {type=SkeletalMeshComponent}
 Timer {type=number}
 TimerHandle {type=TimerHandle}
]
